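# Niagara system for the Galactic Magic drift trail vehicle cosmetic, driven by vehicle flag and speed modules.
# The enabled boost-ribbon emitter spawns paired ribbons at skeletal positions with velocity, source-movement inheritance and painted-variant color overrides.
# It derives several width and position attributes from one particle, using age curves and animated sine offsets for layered cross and lightning strands.
# The mesh-and-ribbon trail and two alternate boost-ribbon emitters are disabled.
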NiagaraSystem /VehicleCosmetics/DriftTrail/DriftTrail_GalacticMagic/FX/NS_DriftTrail_GalacticMagic
dependencies {
 usedDataInterfaces [
  &NiagaraDataInterfaceVectorCurve
  &NiagaraDataInterfaceColorCurve
  &NiagaraDataInterfaceCurve
  &NiagaraDataInterfaceSkeletalMesh
 ]
 usedDynamicInputs []
 usedModules [
  &self:SystemSpawnScript.SystemScriptSource.SystemScriptGraph.NiagaraNodeAssignment_0.SetVariables_#1
  &SystemState
  &FN_UnpackVehicleFlags
  /Game/Effects/Niagara/Modules/Vehicle/FN_UpdateVehicleAttribs
  /VehicleCosmetics/FX/Niagara/Modules/VC_SetVehicleSpeedAttribs
  /Game/Effects/Niagara/Modules/Vehicle/FN_VehicleFlags_Cheatcodes
 ]
}
emitters {
 NE_Trail {
  bEnabled=false
  emitterUpdateScript [
   &EmitterState(Life Cycle Mode=::NewEnumerator1, Loop Duration=1, UseLoopDelay=false)
   &VC_RateQualityHandler(High Spawn Rate=60, Low Spawn Rate=60, Medium Spawn Rate=60, Use Custom Rate?=true)
   &VC_SpawnRibbon_v1(Number of Ribbons=2, SpawnRate=60)
  ]
  particleSpawnScript [
   &VC_PaintedVariantsOverride() {enabled=false}
   &VC_InitializeRibbon_v1(
    Color=<rgba 0.2428 0.2139 0.3325 1>
    Custom Offset=true
    Enable Skeletal Position=true
    Local Vector Offset=&AddVectorToPosition(Position=<-25 0 -25>, Vector=$User.offset)
    Offset Amount=1
    Ribbon Width=20
   )
   &AddVelocity(Velocity=&MakeCustomVectorFromBool(Bool=$User.bIsFrontendPreview, True Vector=<-2000 0 0>, False Vector=<-4000 0 0>)) {module=AddVelocity001}
  ]
  particleUpdateScript [
   &FN_DisableTireFX(Mask Index Attribute:NiagaraInt32=$Particles.SpawnGroup, Mask Index FL=2, Mask Index FR=3, Mask Index RL=0, Mask Index RR=1)
   &ParticleState()
   &InheritSourceMovement(
    Applied Position Scale=&VectorFromCurve(VectorCurve=DataInterface() {propertyValues {
       XCurve {keys [<at 0 1 RCIM_Cubic>, {interpMode=RCIM_Cubic, leaveTangent=-1.655, time=0.7, value=0}]}
       YCurve {keys [<at 0 1 RCIM_Cubic>, {interpMode=RCIM_Cubic, leaveTangent=-1.655, time=0.7, value=0}]}
       ZCurve {keys [<at 0 1 RCIM_Cubic>, {interpMode=RCIM_Cubic, leaveTangent=-1.655, time=0.7, value=0}]}
      }})
    ApplyForce=false
    ApplyPosition=true
    ApplyPositionFalloff=false
   )
   &SolveForcesAndVelocity(Clamp Velocity=false, Limit Acceleration=false)
   &VC_UpdateRibbon_v1(Enable Skeletal Position=true)
   &ScaleColor(
    Linear Color Curve=DataInterface() {propertyValues {
      AlphaCurve {keys [<at 0.3339 1.2>,<at 0.7108 0>]}
      BlueCurve {keys [<at 0.0385 0.1686>,<at 0.104 0.3333>,<at 0.1524 0.8854>,<at 0.2123 0.3325>]}
      GreenCurve {keys [<at 0.0385 0.0762>,<at 0.104 0.1412>,<at 0.1524 0.5027>,<at 0.2123 0.1413>]}
      RedCurve {keys [<at 0.0385 0.0351>,<at 0.104 0.1882>,<at 0.1524 0.5937>,<at 0.2123 0.1878>]}
     }}
    Scale Mode=::NewEnumerator2
   ) {enabled=false}
   &ScaleRibbonWidth(Ribbon Width Scale=2)
   /SetVariables(Particles.Scale=<5 5 5>) {isSetParameters=true}
   &ScaleColor(
    Scale Alpha=&FloatFromCurve(FloatCurve=DataInterface() {propertyValues {Curve {
        defaultValue=0
        keys [
         {interpMode=RCIM_Cubic, tangentMode=RCTM_User, time=0, value=0}
         {arriveTangentWeight=0.1491, interpMode=RCIM_Cubic, leaveTangentWeight=0.2236, tangentMode=RCTM_User, time=0.17, value=0.025}
         {interpMode=RCIM_Cubic, tangentMode=RCTM_User, time=0.5, value=1}
        ]
       }}})
    ScaleA=true
    ScaleRGB=true
   ) {module=ScaleColor001}
  ]
  rendererClasses [&NiagaraMeshRendererProperties, &NiagaraRibbonRendererProperties]
  renderers [
   {rendererClass=&NiagaraMeshRendererProperties, rendererIndex=0}
   {rendererClass=&NiagaraRibbonRendererProperties, rendererIndex=1}
   {rendererClass=&NiagaraRibbonRendererProperties, rendererIndex=2}
   {rendererClass=&NiagaraRibbonRendererProperties, rendererIndex=3}
   {rendererClass=&NiagaraRibbonRendererProperties, rendererIndex=4}
   {rendererClass=&NiagaraRibbonRendererProperties, rendererIndex=5}
  ]
 }
 NE_VC_Boost_Ribbon_Parent {
  emitterUpdateScript [
   &EmitterState()
   &VC_FilteredSocketCount()
   &VC_RateQualityHandler()
   &VC_SpawnRibbon_v1(Number of Ribbons=2, SpawnRate=60)
  ]
  particleSpawnScript [
   &VC_PaintedVariantsOverride(
    Black=<rgba 0.05 0.05 0.05 1>
    BurntSienna=<rgba 0.15 0.0375 0 1>
    Cobalt=<rgba 0.18 0.288 1.8 1>
    Crimson=<rgba 1.3 0.13 0.1617 1>
    DefaultColor=<rgba 0.0259 0.0398 0.1042 0.0174>
    ForestGreen=<rgba 0.2 1 0.2404 1>
    Grey=<rgba 0.28 0.28 0.28 1>
    Lime=<rgba 0.65 1 0.3 1>
    Orange=<rgba 1.2 0.5 0 1>
    Pink=<rgba 1 0.25 0.7273 1>
    Purple=<rgba 0.4 0.2 0.8 1>
    R1 Black:LinearColor=&Multiply_LinearColorByFloat(Linear Color=<rgba 0 0 0 1>, Float=3)
    R1 BurntSienna:LinearColor=<rgba 0.013 0.004 0.0016 3>
    R1 Cobalt:LinearColor=<rgba 0 0.0185 1 0.5>
    R1 Crimson:LinearColor=<rgba 0.464 0.1392 0.148 0.5>
    R1 DefaultColor=&Multiply_LinearColorByFloat(Linear Color=<rgba 1 1 1 1>, Float=2)
    R1 ForestGreen:LinearColor=<rgba 0 0.5 0.0253 0.5>
    R1 Grey:LinearColor=<rgba 0.1 0.1 0.1 2>
    R1 Lime:LinearColor=<rgba 0.2943 0.5885 0 0.2>
    R1 Orange:LinearColor=<rgba 0.6 0.15 0 0.8>
    R1 Pink:LinearColor=<rgba 1 0.3 0.7455 0.2>
    R1 Purple:LinearColor=<rgba 0.0787 0.0213 0.71 1.3>
    R1 Saffron:LinearColor=<rgba 0.7 0.5833 0 0.5>
    R1 SkyBlue:LinearColor=<rgba 0 0.357 0.5365 0.25>
    R1 TitaniumWhite:LinearColor=<rgba 0.5 0.5 0.5 0.35>
    R2 Black:LinearColor=&Multiply_LinearColorByFloat(Linear Color=<rgba 0 0 0 1>, Float=3)
    R2 BurntSienna:LinearColor=<rgba 0.03 0.0101 0.006 1>
    R2 Cobalt:LinearColor=<rgba 0 0.025 1.5 0.5>
    R2 Crimson:LinearColor=<rgba 0.6 0.03 0.049 0.7>
    R2 DefaultColor:LinearColor=<rgba 0 2.8619 4 0.2261>
    R2 ForestGreen:LinearColor=<rgba 0 0.6 0.0303 1>
    R2 Grey:LinearColor=<rgba 0.14 0.14 0.14 0.8>
    R2 Lime:LinearColor=<rgba 0.315 0.6 0.03 1>
    R2 Orange:LinearColor=<rgba 0.9 0.165 0 0.8>
    R2 Pink:LinearColor=<rgba 0.8 0.096 0.544 0.8>
    R2 Purple:LinearColor=<rgba 0.0875 0.05 0.5 0.4>
    R2 Saffron:LinearColor=<rgba 1 0.8333 0 0.75>
    R2 SkyBlue:LinearColor=<rgba 0 0.5 1 0.7>
    R2 TitaniumWhite:LinearColor=<rgba 2 2 2 0.25>
    Saffron=<rgba 1.2 1.004 0.36 1>
    SkyBlue=<rgba 0.25 0.749 1 1>
    TitaniumWhite=<rgba 0.3 0.3 0.3 1>
    Use R1 Painted Overrides=true
    Use R2 Painted Overrides=true
   )
   &VC_InitializeRibbon_v1(
    Color=$Particles.Color
    Custom Offset=true
    Enable Skeletal Position=true
    Lifetime=0.8
    Local Vector Offset=&AddVectorToPosition(Position=<-70 0 0>, Vector=Linked() {linkedVariable {name=User.DTCarOffset, type=NiagaraPosition}})
    Offset Amount=1
    Ribbon Width=60
    Write Material Random=false
   )
   &AddVelocity(Velocity=<-2000 0 0>) {module=AddVelocity001}
  ]
  particleUpdateScript [
   &ParticleState()
   &InheritSourceMovement(ApplyForce=false, ApplyPosition=true, ApplyPositionFalloff=false)
   &SolveForcesAndVelocity(Clamp Velocity=false, Limit Acceleration=false)
   &VC_UpdateRibbon_v1(Enable Skeletal Position=true)
   &ScaleColor(
    Scale Alpha=&FloatFromCurve(
     FloatCurve=DataInterface() {propertyValues {Curve {
        defaultValue=1
        keys [
         {arriveTangent=3.8118, interpMode=RCIM_Cubic, leaveTangent=3.8118, tangentMode=RCTM_User, time=0, value=0}
         {arriveTangent=1.5427, interpMode=RCIM_Cubic, leaveTangent=-0.033, tangentMode=RCTM_Break, time=0.1789, value=0.9847}
         {arriveTangent=-5.5211, interpMode=RCIM_Cubic, leaveTangent=-5.5211, tangentMode=RCTM_User, time=0.8734, value=0}
        ]
       }}}
     Scale Curve=20
    )
    ScaleA=true
    ScaleRGB=false
   )
   /SetVariables(
    Particles.BlueWidth:NiagaraFloat=&FloatFromCurve(
     FloatCurve=DataInterface() {propertyValues {Curve {
        defaultValue=1
        keys [
         <at -0.0085 0.7008 RCIM_Cubic>
         {arriveTangent=0.0448, interpMode=RCIM_Cubic, leaveTangent=0.0448, tangentMode=RCTM_User, time=0.2772, value=0.9847}
         {arriveTangent=-0.932, interpMode=RCIM_Cubic, leaveTangent=-0.932, tangentMode=RCTM_User, time=1.0722, value=0.538}
        ]
       }}}
     Scale Curve=150
    )
    Particles.CrossInvertWidth:NiagaraFloat=&FloatFromCurve(
     FloatCurve=DataInterface() {propertyValues {Curve {
        defaultValue=1
        keys [
         <at -0.0085 0.7008 RCIM_Cubic>
         {arriveTangent=0.0448, interpMode=RCIM_Cubic, leaveTangent=0.0448, tangentMode=RCTM_User, time=0.2772, value=0.9847}
         {arriveTangent=-0.0405, interpMode=RCIM_Cubic, leaveTangent=-0.0405, tangentMode=RCTM_User, time=0.8385, value=0.0052}
        ]
       }}}
     Scale Curve=10
    )
    Particles.CrossWidth:NiagaraFloat=&FloatFromCurve(
     FloatCurve=DataInterface() {propertyValues {Curve {
        defaultValue=1
        keys [
         <at -0.0085 0.7008 RCIM_Cubic>
         {arriveTangent=0.0448, interpMode=RCIM_Cubic, leaveTangent=0.0448, tangentMode=RCTM_User, time=0.2772, value=0.9847}
         {arriveTangent=-0.3044, interpMode=RCIM_Cubic, leaveTangent=-0.3044, tangentMode=RCTM_User, time=0.8, value=0}
        ]
       }}}
     Scale Curve=50
    )
    Particles.ThunderWidth:NiagaraFloat=&FloatFromCurve(
     FloatCurve=DataInterface() {propertyValues {Curve {
        defaultValue=1
        keys [
         {arriveTangent=1.4927, interpMode=RCIM_Cubic, time=-0.0449, value=0.734}
         {arriveTangent=1.5427, interpMode=RCIM_Cubic, leaveTangent=-0.1679, tangentMode=RCTM_Break, time=0.1789, value=0.9847}
         {arriveTangent=-2.714, interpMode=RCIM_Cubic, leaveTangent=-2.714, tangentMode=RCTM_User, time=1.0781, value=-0.1679}
        ]
       }}}
     Scale Curve=50
    )
   ) {isSetParameters=true, module=SetVariables_#2}
   /SetVariables(
    Particles.CrossInvertPosition:NiagaraPosition=&AddVectorToPosition(
     Position=$Particles.Position
     Vector=&TransformVector(
      Vector=&MakeVector(
       X=&SelectFloatFromArray(Direct Array Index=&SpawnGroup(), Float Selection Array=DataInterface() {propertyValues {FloatData [-30,0]}})
       Y=0
       Z=&Sine(
        Normalized Angle=&Add_Float(
         A=&Multiply_Float(A=$Particles.NormalizedAge, B=-1.3)
         B=&Multiply_Float(A=$Emitter.LoopedAge, B=1.2)
        )
        Period=-1
        Scale=&FloatFromCurve(
         FloatCurve=DataInterface() {propertyValues {Curve {
            defaultValue=1
            keys [
             {arriveTangent=2.952, interpMode=RCIM_Cubic, leaveTangent=2.952, tangentMode=RCTM_User, time=-0.0185, value=0}
             {arriveTangent=0.1401, tangentMode=RCTM_User, time=0.387, value=0.6595}
             {arriveTangent=-1.5773, interpMode=RCIM_Cubic, leaveTangent=-3.8709, tangentMode=RCTM_Break, time=0.5819, value=0.6595}
             {arriveTangent=-0.2749, interpMode=RCIM_Cubic, leaveTangent=-0.2749, tangentMode=RCTM_User, time=1.0012, value=0}
            ]
           }}}
         Scale Curve=20
        )
        Bias=0
       )
      )
      Destination Space=::Simulation
      Apply Scale=true
     )
    )
    Particles.PositionCross:NiagaraPosition=&AddVectorToPosition(
     Position=$Particles.Position
     Vector=&TransformVector(
      Vector=&MakeVector(
       X=&SelectFloatFromArray(Direct Array Index=&SpawnGroup(), Float Selection Array=DataInterface() {propertyValues {FloatData [-30,0]}})
       Y=0
       Z=&Sine(
        Normalized Angle=&Add_Float(
         A=&Multiply_Float(A=$Particles.NormalizedAge, B=-1.3)
         B=&Multiply_Float(A=$Emitter.LoopedAge, B=1.5)
        )
        Period=1
        Scale=&FloatFromCurve(
         FloatCurve=DataInterface() {propertyValues {Curve {
            defaultValue=1
            keys [
             {arriveTangent=-0.1097, interpMode=RCIM_Cubic, leaveTangent=-0.1097, tangentMode=RCTM_User, time=0.175, value=0}
             {arriveTangent=7.2017, leaveTangent=7.2017, time=0.3137, value=0.7182}
             {arriveTangent=-2.5934, interpMode=RCIM_Cubic, leaveTangent=-2.5934, tangentMode=RCTM_User, time=0.4951, value=0.7335}
             {arriveTangent=-0.0405, interpMode=RCIM_Cubic, leaveTangent=-0.0405, tangentMode=RCTM_User, time=0.7, value=0}
            ]
           }}}
         Scale Curve=15
        )
        Bias=0
       )
      )
      Destination Space=::Simulation
      Apply Scale=true
     )
    )
    Particles.PositionLightning:NiagaraPosition=&AddVectorToPosition(
     Position=$Particles.Position
     Vector=&TransformVector(
      Vector=&MakeVector(
       X=0
       Y=0
       Z=&Sine(
        Normalized Angle=&Add_Float(
         A=&Multiply_Float(A=$Particles.NormalizedAge, B=-1.3)
         B=&Multiply_Float(A=$Emitter.LoopedAge, B=1.3)
        )
        Period=1
        Scale=&FloatFromCurve(
         FloatCurve=DataInterface() {propertyValues {Curve {
            defaultValue=1
            keys [
             {arriveTangent=9.0725, interpMode=RCIM_Cubic, leaveTangent=9.0725, tangentMode=RCTM_User, time=-0.0048, value=0}
             {arriveTangent=0.0244, interpMode=RCIM_Cubic, leaveTangent=0.0244, tangentMode=RCTM_User, time=0.1408, value=0.9847}
             {arriveTangent=-0.0583, interpMode=RCIM_Cubic, leaveTangent=-0.0583, tangentMode=RCTM_User, time=0.3728, value=1}
             {arriveTangent=-0.2024, interpMode=RCIM_Cubic, leaveTangent=-0.2024, tangentMode=RCTM_User, time=1.1034, value=0}
            ]
           }}}
         Scale Curve=13
        )
        Bias=0
       )
      )
      Destination Space=::Simulation
      Apply Scale=true
     )
    )
   ) {isSetParameters=true, module=SetVariables_#3}
  ]
  rendererClasses [&NiagaraRibbonRendererProperties]
  renderers [
   {rendererClass=&NiagaraRibbonRendererProperties}
   {rendererClass=&NiagaraRibbonRendererProperties, rendererIndex=1}
   {rendererClass=&NiagaraRibbonRendererProperties, rendererIndex=2}
   {rendererClass=&NiagaraRibbonRendererProperties, rendererIndex=3}
  ]
 }
 NE_VC_Boost_Ribbon_Parent001 {
  bEnabled=false
  emitterUpdateScript [
   &EmitterState()
   &VC_FilteredSocketCount()
   &VC_RateQualityHandler()
   &VC_SpawnRibbon_v1(Number of Ribbons=2, SpawnRate=60)
  ]
  particleSpawnScript [
   &VC_PaintedVariantsOverride(
    Black=<rgba 0.05 0.05 0.05 1>
    BurntSienna=<rgba 0.15 0.0375 0 1>
    Cobalt=<rgba 0.18 0.288 1.8 1>
    Crimson=<rgba 1.3 0.13 0.1617 1>
    DefaultColor=<rgba 1 0.7783 0.05 0.5913>
    ForestGreen=<rgba 0.2 1 0.2404 1>
    Grey=<rgba 0.28 0.28 0.28 1>
    Lime=<rgba 0.65 1 0.3 1>
    Orange=<rgba 1.2 0.5 0 1>
    Pink=<rgba 1 0.25 0.7273 1>
    Purple=<rgba 0.4 0.2 0.8 1>
    R1 Black:LinearColor=&Multiply_LinearColorByFloat(Linear Color=<rgba 0 0 0 1>, Float=3)
    R1 BurntSienna:LinearColor=<rgba 0.013 0.004 0.0016 3>
    R1 Cobalt:LinearColor=<rgba 0 0.0185 1 0.5>
    R1 Crimson:LinearColor=<rgba 0.464 0.1392 0.148 0.5>
    R1 DefaultColor=&Multiply_LinearColorByFloat(Linear Color=<rgba 1 1 1 1>, Float=2)
    R1 ForestGreen:LinearColor=<rgba 0 0.5 0.0253 0.5>
    R1 Grey:LinearColor=<rgba 0.1 0.1 0.1 2>
    R1 Lime:LinearColor=<rgba 0.2943 0.5885 0 0.2>
    R1 Orange:LinearColor=<rgba 0.6 0.15 0 0.8>
    R1 Pink:LinearColor=<rgba 1 0.3 0.7455 0.2>
    R1 Purple:LinearColor=<rgba 0.0787 0.0213 0.71 1.3>
    R1 Saffron:LinearColor=<rgba 0.7 0.5833 0 0.5>
    R1 SkyBlue:LinearColor=<rgba 0 0.357 0.5365 0.25>
    R1 TitaniumWhite:LinearColor=<rgba 0.5 0.5 0.5 0.35>
    R2 Black:LinearColor=&Multiply_LinearColorByFloat(Linear Color=<rgba 0 0 0 1>, Float=3)
    R2 BurntSienna:LinearColor=<rgba 0.03 0.0101 0.006 1>
    R2 Cobalt:LinearColor=<rgba 0 0.025 1.5 0.5>
    R2 Crimson:LinearColor=<rgba 0.6 0.03 0.049 0.7>
    R2 DefaultColor:LinearColor=<rgba 0 2.8619 4 0.2261>
    R2 ForestGreen:LinearColor=<rgba 0 0.6 0.0303 1>
    R2 Grey:LinearColor=<rgba 0.14 0.14 0.14 0.8>
    R2 Lime:LinearColor=<rgba 0.315 0.6 0.03 1>
    R2 Orange:LinearColor=<rgba 0.9 0.165 0 0.8>
    R2 Pink:LinearColor=<rgba 0.8 0.096 0.544 0.8>
    R2 Purple:LinearColor=<rgba 0.0875 0.05 0.5 0.4>
    R2 Saffron:LinearColor=<rgba 1 0.8333 0 0.75>
    R2 SkyBlue:LinearColor=<rgba 0 0.5 1 0.7>
    R2 TitaniumWhite:LinearColor=<rgba 2 2 2 0.25>
    Saffron=<rgba 1.2 1.004 0.36 1>
    SkyBlue=<rgba 0.25 0.749 1 1>
    TitaniumWhite=<rgba 0.3 0.3 0.3 1>
    Use R1 Painted Overrides=true
    Use R2 Painted Overrides=true
   )
   &VC_InitializeRibbon_v1(
    Color=$Particles.Color
    Custom Offset=true
    Enable Skeletal Position=true
    Lifetime=0.8
    Local Vector Offset=&AddVectorToPosition(Position=<-70 0 0>, Vector=Linked() {linkedVariable {name=User.DTCarOffset, type=NiagaraPosition}})
    Offset Amount=1
    Ribbon Width=60
    Write Material Random=false
   )
   &AddVelocity(Velocity=<-2000 0 0>) {module=AddVelocity001}
  ]
  particleUpdateScript [
   &ParticleState()
   &InheritSourceMovement(ApplyForce=false, ApplyPosition=true, ApplyPositionFalloff=false)
   &SolveForcesAndVelocity(Clamp Velocity=false, Limit Acceleration=false)
   &VC_UpdateRibbon_v1(Enable Skeletal Position=true)
   &ScaleColor(
    Scale Alpha=&FloatFromCurve(
     FloatCurve=DataInterface() {propertyValues {Curve {
        defaultValue=1
        keys [
         {arriveTangent=3.8118, interpMode=RCIM_Cubic, leaveTangent=3.8118, tangentMode=RCTM_User, time=0, value=0}
         {arriveTangent=1.5427, interpMode=RCIM_Cubic, leaveTangent=-0.033, tangentMode=RCTM_Break, time=0.1789, value=0.9847}
         {arriveTangent=-5.5211, interpMode=RCIM_Cubic, leaveTangent=-5.5211, tangentMode=RCTM_User, time=0.8734, value=0}
        ]
       }}}
     Scale Curve=20
    )
    ScaleA=true
    ScaleRGB=false
   )
   /SetVariables(
    Particles.BeamColor:LinearColor=&MultiplyLinearColor(
     Linear Color=<rgba 1 1 1 1>
     Scale Factor=&MakeVector4FromVectorAndFloat(
      "Vector (RGB)"=<1 1 1>
      "Float (Alpha)"=&FloatFromCurve(FloatCurve=DataInterface() {propertyValues {Curve {
          defaultValue=1
          keys [
           {arriveTangent=1.4927, interpMode=RCIM_Cubic, time=-0.0449, value=0.734}
           {arriveTangent=1.5427, interpMode=RCIM_Cubic, leaveTangent=-1.0354, tangentMode=RCTM_Break, time=0.1789, value=0.9847}
           {arriveTangent=-3.3735, interpMode=RCIM_Cubic, leaveTangent=-3.3735, tangentMode=RCTM_User, time=1.0198, value=-0.2103}
          ]
         }}})
     )
    )
    Particles.BlueWidth:NiagaraFloat=&FloatFromCurve(
     FloatCurve=DataInterface() {propertyValues {Curve {
        defaultValue=1
        keys [
         <at -0.0085 0.7008 RCIM_Cubic>
         {arriveTangent=0.0448, interpMode=RCIM_Cubic, leaveTangent=0.0448, tangentMode=RCTM_User, time=0.2772, value=0.9847}
         {arriveTangent=-0.932, interpMode=RCIM_Cubic, leaveTangent=-0.932, tangentMode=RCTM_User, time=1.0722, value=0.538}
        ]
       }}}
     Scale Curve=150
    )
    Particles.CrossColor:LinearColor=&MakeCustomLinearColorFromBool(
     Bool=$User.bIsPainted
     True Color=&MultiplyLinearColor(
      Linear Color=<rgba 1 1 1 1>
      Scale Factor=&MakeVector4FromVectorAndFloat(
       "Vector (RGB)"=<1 1 1>
       "Float (Alpha)"=&FloatFromCurve(
        FloatCurve=DataInterface() {propertyValues {Curve {
           defaultValue=1
           keys [
            {arriveTangent=1.4927, interpMode=RCIM_Cubic, time=0, value=0}
            {arriveTangent=0.1055, interpMode=RCIM_Cubic, leaveTangent=-1.0354, tangentMode=RCTM_Break, time=0.2653, value=1}
            {arriveTangent=-0.2248, interpMode=RCIM_Cubic, leaveTangent=-0.2248, tangentMode=RCTM_User, time=0.9436, value=0}
           ]
          }}}
        Scale Curve=3
       )
      )
     )
     False Color=&ColorFromCurve(ColorCurve=DataInterface() {propertyValues {
        AlphaCurve {keys [<at 0 0 RCIM_Cubic>,<at 0.4186 1>,<at 1 2 RCIM_Cubic>]}
        BlueCurve {keys [<at 0 1 RCIM_Cubic>,<at 0.7607 1 RCIM_Cubic>]}
        GreenCurve {keys [<at 0 0.6654 RCIM_Cubic>,<at 0.7607 0 RCIM_Cubic>]}
        RedCurve {keys [<at 0 0 RCIM_Cubic>,<at 0.7607 0.029 RCIM_Cubic>]}
       }})
    )
    Particles.CrossColor2:LinearColor=&MakeCustomLinearColorFromBool(
     Bool=$User.bIsFrontendPreview
     True Color=&MultiplyLinearColor(
      Linear Color=<rgba 1 1 1 1>
      Scale Factor=&MakeVector4FromVectorAndFloat(
       "Vector (RGB)"=<1 1 1>
       "Float (Alpha)"=&FloatFromCurve(FloatCurve=DataInterface() {propertyValues {Curve {
           defaultValue=1
           keys [
            {arriveTangent=1.4927, interpMode=RCIM_Cubic, time=-0.0449, value=0.734}
            {arriveTangent=1.5427, interpMode=RCIM_Cubic, leaveTangent=-1.0354, tangentMode=RCTM_Break, time=0.1789, value=0.9847}
            {arriveTangent=-3.3735, interpMode=RCIM_Cubic, leaveTangent=-3.3735, tangentMode=RCTM_User, time=1.0198, value=-0.2103}
           ]
          }}})
      )
     )
     False Color=&MultiplyLinearColor(
      Linear Color=$Particles.R1 Color
      Scale Factor=&MakeVector4FromVectorAndFloat(
       "Vector (RGB)"=<1 5 2>
       "Float (Alpha)"=&FloatFromCurve(
        FloatCurve=DataInterface() {propertyValues {Curve {
           defaultValue=1
           keys [
            {arriveTangent=1.4927, interpMode=RCIM_Cubic, time=-0.0449, value=0.734}
            {arriveTangent=1.5427, interpMode=RCIM_Cubic, leaveTangent=-1.0354, tangentMode=RCTM_Break, time=0.1789, value=0.9847}
            {arriveTangent=-3.3735, interpMode=RCIM_Cubic, leaveTangent=-3.3735, tangentMode=RCTM_User, time=1.0198, value=-0.2103}
           ]
          }}}
        Scale Curve=0.5
       )
      )
     )
    )
    Particles.CrossInvertWidth:NiagaraFloat=&FloatFromCurve(
     FloatCurve=DataInterface() {propertyValues {Curve {
        defaultValue=1
        keys [
         <at -0.0085 0.7008 RCIM_Cubic>
         {arriveTangent=0.0448, interpMode=RCIM_Cubic, leaveTangent=0.0448, tangentMode=RCTM_User, time=0.2772, value=0.9847}
         {arriveTangent=-0.0405, interpMode=RCIM_Cubic, leaveTangent=-0.0405, tangentMode=RCTM_User, time=0.8385, value=0.0052}
        ]
       }}}
     Scale Curve=10
    )
    Particles.CrossWidth:NiagaraFloat=&FloatFromCurve(
     FloatCurve=DataInterface() {propertyValues {Curve {
        defaultValue=1
        keys [
         <at -0.0085 0.7008 RCIM_Cubic>
         {arriveTangent=0.0448, interpMode=RCIM_Cubic, leaveTangent=0.0448, tangentMode=RCTM_User, time=0.2772, value=0.9847}
         {arriveTangent=-0.3044, interpMode=RCIM_Cubic, leaveTangent=-0.3044, tangentMode=RCTM_User, time=0.8, value=0}
        ]
       }}}
     Scale Curve=50
    )
    Particles.LightningColor:LinearColor=&MultiplyLinearColor(
     Linear Color=<rgba 1 1 1 1>
     Scale Factor=&MakeVector4FromVectorAndFloat(
      "Vector (RGB)"=<1 1 1>
      "Float (Alpha)"=&FloatFromCurve(FloatCurve=DataInterface() {propertyValues {Curve {
          defaultValue=1
          keys [
           {arriveTangent=0.0901, interpMode=RCIM_Cubic, leaveTangent=0.0901, tangentMode=RCTM_User, time=-0.0449, value=-0.0307}
           {arriveTangent=-0.4524, interpMode=RCIM_Cubic, leaveTangent=0.0385, tangentMode=RCTM_Break, time=0.5314, value=0.9847}
           {arriveTangent=-0.1302, interpMode=RCIM_Cubic, leaveTangent=-0.1302, tangentMode=RCTM_User, time=1.4722, value=0}
          ]
         }}})
     )
    )
    Particles.ThunderWidth:NiagaraFloat=&FloatFromCurve(
     FloatCurve=DataInterface() {propertyValues {Curve {
        defaultValue=1
        keys [
         {arriveTangent=1.4927, interpMode=RCIM_Cubic, time=-0.0449, value=0.734}
         {arriveTangent=1.5427, interpMode=RCIM_Cubic, leaveTangent=-0.1679, tangentMode=RCTM_Break, time=0.1789, value=0.9847}
         {arriveTangent=-2.714, interpMode=RCIM_Cubic, leaveTangent=-2.714, tangentMode=RCTM_User, time=1.0781, value=-0.1679}
        ]
       }}}
     Scale Curve=50
    )
   ) {isSetParameters=true, module=SetVariables_#2}
   /SetVariables(
    Particles.CrossInvertPosition:NiagaraPosition=&AddVectorToPosition(
     Position=$Particles.Position
     Vector=&TransformVector(
      Vector=&MakeVector(
       X=&SelectFloatFromArray(Direct Array Index=&SpawnGroup(), Float Selection Array=DataInterface() {propertyValues {FloatData [-30,0]}})
       Y=0
       Z=&Sine(
        Normalized Angle=&Add_Float(
         A=&Multiply_Float(A=$Particles.NormalizedAge, B=-1.3)
         B=&Multiply_Float(A=$Emitter.LoopedAge, B=1.2)
        )
        Period=-1
        Scale=&FloatFromCurve(
         FloatCurve=DataInterface() {propertyValues {Curve {
            defaultValue=1
            keys [
             {arriveTangent=2.952, interpMode=RCIM_Cubic, leaveTangent=2.952, tangentMode=RCTM_User, time=-0.0185, value=0}
             {arriveTangent=0.1401, tangentMode=RCTM_User, time=0.387, value=0.6595}
             {arriveTangent=-1.5773, interpMode=RCIM_Cubic, leaveTangent=-3.8709, tangentMode=RCTM_Break, time=0.5819, value=0.6595}
             {arriveTangent=-0.2749, interpMode=RCIM_Cubic, leaveTangent=-0.2749, tangentMode=RCTM_User, time=1.0012, value=0}
            ]
           }}}
         Scale Curve=20
        )
        Bias=0
       )
      )
      Destination Space=::Simulation
      Apply Scale=true
     )
    )
    Particles.PositionCross:NiagaraPosition=&AddVectorToPosition(
     Position=$Particles.Position
     Vector=&TransformVector(
      Vector=&MakeVector(
       X=&SelectFloatFromArray(Direct Array Index=&SpawnGroup(), Float Selection Array=DataInterface() {propertyValues {FloatData [-30,0]}})
       Y=0
       Z=&Sine(
        Normalized Angle=&Add_Float(
         A=&Multiply_Float(A=$Particles.NormalizedAge, B=-1.3)
         B=&Multiply_Float(A=$Emitter.LoopedAge, B=1.5)
        )
        Period=1
        Scale=&FloatFromCurve(
         FloatCurve=DataInterface() {propertyValues {Curve {
            defaultValue=1
            keys [
             {arriveTangent=-0.1097, interpMode=RCIM_Cubic, leaveTangent=-0.1097, tangentMode=RCTM_User, time=0.175, value=0}
             {arriveTangent=7.2017, leaveTangent=7.2017, time=0.3137, value=0.7182}
             {arriveTangent=-2.5934, interpMode=RCIM_Cubic, leaveTangent=-2.5934, tangentMode=RCTM_User, time=0.4951, value=0.7335}
             {arriveTangent=-0.0405, interpMode=RCIM_Cubic, leaveTangent=-0.0405, tangentMode=RCTM_User, time=0.7, value=0}
            ]
           }}}
         Scale Curve=15
        )
        Bias=0
       )
      )
      Destination Space=::Simulation
      Apply Scale=true
     )
    )
    Particles.PositionLightning:NiagaraPosition=&AddVectorToPosition(
     Position=$Particles.Position
     Vector=&TransformVector(
      Vector=&MakeVector(
       X=0
       Y=0
       Z=&Sine(
        Normalized Angle=&Add_Float(
         A=&Multiply_Float(A=$Particles.NormalizedAge, B=-1.3)
         B=&Multiply_Float(A=$Emitter.LoopedAge, B=1.3)
        )
        Period=1
        Scale=&FloatFromCurve(
         FloatCurve=DataInterface() {propertyValues {Curve {
            defaultValue=1
            keys [
             {arriveTangent=9.0725, interpMode=RCIM_Cubic, leaveTangent=9.0725, tangentMode=RCTM_User, time=-0.0048, value=0}
             {arriveTangent=0.0244, interpMode=RCIM_Cubic, leaveTangent=0.0244, tangentMode=RCTM_User, time=0.1408, value=0.9847}
             {arriveTangent=-0.0583, interpMode=RCIM_Cubic, leaveTangent=-0.0583, tangentMode=RCTM_User, time=0.3728, value=1}
             {arriveTangent=-0.2024, interpMode=RCIM_Cubic, leaveTangent=-0.2024, tangentMode=RCTM_User, time=1.1034, value=0}
            ]
           }}}
         Scale Curve=13
        )
        Bias=0
       )
      )
      Destination Space=::Simulation
      Apply Scale=true
     )
    )
   ) {isSetParameters=true, module=SetVariables_#3}
  ]
  rendererClasses [&NiagaraRibbonRendererProperties]
  renderers [
   {rendererClass=&NiagaraRibbonRendererProperties}
   {rendererClass=&NiagaraRibbonRendererProperties, rendererIndex=1}
   {rendererClass=&NiagaraRibbonRendererProperties, rendererIndex=2}
   {rendererClass=&NiagaraRibbonRendererProperties, rendererIndex=3}
  ]
 }
 NE_VC_Boost_Ribbon_Parent002 {
  bEnabled=false
  emitterUpdateScript [
   &EmitterState()
   &VC_FilteredSocketCount()
   &VC_RateQualityHandler()
   &VC_SpawnRibbon_v1(Number of Ribbons=2, SpawnRate=60)
  ]
  particleSpawnScript [
   &VC_PaintedVariantsOverride(Crimson=<rgba 0.4635 0 0.0126 1>, Purple=<rgba 0.147 0.0176 0.6795 1>)
   &VC_InitializeRibbon_v1(
    Color=$Particles.Color
    Custom Offset=true
    Enable Skeletal Position=true
    Lifetime=0.8
    Local Vector Offset=&AddVectorToPosition(Position=<-40 0 0>, Vector=$User.DTCarOffset001)
    Offset Amount=1
    Ribbon Width=40
   )
   &AddVelocity(Velocity=<-1500 0 0>, Velocity Speed Scale=2) {module=AddVelocity001}
  ]
  particleUpdateScript [
   &ParticleState()
   &VC_PaintedVariantsOverride(
    BurntSienna=<rgba 0.0642 0.0198 0.0079 1>
    Cobalt=<rgba 0 0.0185 1 1>
    Crimson=<rgba 0.4635 0 0.0126 1>
    ForestGreen=<rgba 0 0.5 0.0253 1>
    Grey=<rgba 0.0781 0.0781 0.0781 1>
    Lime=<rgba 0.2943 0.5885 0 1>
    Orange=<rgba 0.6724 0.2346 0 1>
    Purple=<rgba 0.0844 0.0482 0.3372 1>
    R1 Black:LinearColor=&ColorFromCurve(ColorCurve=DataInterface() {propertyValues {
       AlphaCurve {keys [<at 0 1 RCIM_Cubic>,<at 1 0 RCIM_Cubic>]}
       BlueCurve {keys [<at 0 0.0521 RCIM_Cubic>,<at 0.1166 0.01 RCIM_Cubic>]}
       GreenCurve {keys [<at 0 0.0521 RCIM_Cubic>,<at 0.1166 0.01 RCIM_Cubic>]}
       RedCurve {keys [<at 0 0.0521 RCIM_Cubic>,<at 0.1166 0.01 RCIM_Cubic>]}
      }})
    R1 BurntSienna:LinearColor=&ColorFromCurve(ColorCurve=DataInterface() {propertyValues {
       AlphaCurve {keys [<at 0 1 RCIM_Cubic>,<at 1 0 RCIM_Cubic>]}
       BlueCurve {keys [<at 0 0.0482 RCIM_Cubic>,<at 0.2929 0.0107 RCIM_Cubic>]}
       GreenCurve {keys [<at 0 0.0887 RCIM_Cubic>,<at 0.2929 0.0196 RCIM_Cubic>]}
       RedCurve {keys [<at 0 0.4233 RCIM_Cubic>,<at 0.2929 0.0938 RCIM_Cubic>]}
      }})
    R1 Cobalt:LinearColor=&ColorFromCurve(ColorCurve=DataInterface() {propertyValues {
       AlphaCurve {keys [<at 0 1 RCIM_Cubic>,<at 1 0 RCIM_Cubic>]}
       BlueCurve {keys [<at 0 0.3594 RCIM_Cubic>,<at 0.2929 0.7231 RCIM_Cubic>]}
       GreenCurve {keys [<at 0 0.115 RCIM_Cubic>,<at 0.2929 0.1022 RCIM_Cubic>]}
       RedCurve {keys [<at 0 0.1104 RCIM_Cubic>,<at 0.2929 0.0056 RCIM_Cubic>]}
      }})
    R1 Crimson:LinearColor=&ColorFromCurve(ColorCurve=DataInterface() {propertyValues {
       AlphaCurve {keys [<at 0 1 RCIM_Cubic>,<at 1 0 RCIM_Cubic>]}
       BlueCurve {keys [<at 0 0.0755 RCIM_Cubic>,<at 0.2929 0.0123 RCIM_Cubic>]}
       GreenCurve {keys [<at 0 0.065 RCIM_Cubic>,<at 0.2929 0 RCIM_Cubic>]}
       RedCurve {keys [<at 0 0.4621 RCIM_Cubic>,<at 0.2929 0.4621 RCIM_Cubic>]}
      }})
    R1 ForestGreen:LinearColor=&ColorFromCurve(ColorCurve=DataInterface() {propertyValues {
       AlphaCurve {keys [<at 0 1 RCIM_Cubic>,<at 1 0 RCIM_Cubic>]}
       BlueCurve {keys [<at 0 0.1981 RCIM_Cubic>,<at 0.2929 0.028 RCIM_Cubic>]}
       GreenCurve {keys [<at 0 0.5781 RCIM_Cubic>,<at 0.2929 0.5029 RCIM_Cubic>]}
       RedCurve {keys [<at 0 0.2223 RCIM_Cubic>,<at 0.2929 0.1087 RCIM_Cubic>]}
      }})
    R1 Grey:LinearColor=&ColorFromCurve(ColorCurve=DataInterface() {propertyValues {
       AlphaCurve {keys [<at 0 1 RCIM_Cubic>,<at 1 0 RCIM_Cubic>]}
       BlueCurve {keys [<at 0 0.151 RCIM_Cubic>,<at 0.2929 0.0521 RCIM_Cubic>]}
       GreenCurve {keys [<at 0 0.151 RCIM_Cubic>,<at 0.2929 0.0521 RCIM_Cubic>]}
       RedCurve {keys [<at 0 0.151 RCIM_Cubic>,<at 0.2929 0.0521 RCIM_Cubic>]}
      }})
    R1 Lime:LinearColor=&ColorFromCurve(ColorCurve=DataInterface() {propertyValues {
       AlphaCurve {keys [<at 0 1 RCIM_Cubic>,<at 1 0 RCIM_Cubic>]}
       BlueCurve {keys [<at 0 0.0431 RCIM_Cubic>,<at 0.2126 0 RCIM_Cubic>]}
       GreenCurve {keys [<at 0 0.5906 RCIM_Cubic>,<at 0.2126 0.3958 RCIM_Cubic>]}
       RedCurve {keys [<at 0 0.3176 RCIM_Cubic>,<at 0.2126 0.1985 RCIM_Cubic>]}
      }})
    R1 Orange:LinearColor=&ColorFromCurve(ColorCurve=DataInterface() {propertyValues {
       AlphaCurve {keys [<at 0 1 RCIM_Cubic>,<at 1 0 RCIM_Cubic>]}
       BlueCurve {keys [<at 0 0 RCIM_Cubic>,<at 0.2126 0.0028 RCIM_Cubic>]}
       GreenCurve {keys [<at 0 0.2335 RCIM_Cubic>,<at 0.2126 0.0221 RCIM_Cubic>]}
       RedCurve {keys [<at 0 0.707 RCIM_Cubic>,<at 0.2126 0.2917 RCIM_Cubic>]}
      }})
    R1 Pink:LinearColor=&ColorFromCurve(ColorCurve=DataInterface() {propertyValues {
       AlphaCurve {keys [<at 0 1 RCIM_Cubic>,<at 1 0 RCIM_Cubic>]}
       BlueCurve {keys [<at 0 0.237 RCIM_Cubic>,<at 0.2126 0.3419 RCIM_Cubic>]}
       GreenCurve {keys [<at 0 0.1507 RCIM_Cubic>,<at 0.2126 0.0612 RCIM_Cubic>]}
       RedCurve {keys [<at 0 0.2865 RCIM_Cubic>,<at 0.2126 0.5029 RCIM_Cubic>]}
      }})
    R1 Purple:LinearColor=&ColorFromCurve(ColorCurve=DataInterface() {propertyValues {
       AlphaCurve {keys [<at 0 1 RCIM_Cubic>,<at 1 0 RCIM_Cubic>]}
       BlueCurve {keys [<at 0 0.3372 RCIM_Cubic>,<at 0.2126 0.4323 RCIM_Cubic>]}
       GreenCurve {keys [<at 0 0.1106 RCIM_Cubic>,<at 0.2126 0.0316 RCIM_Cubic>]}
       RedCurve {keys [<at 0 0.139 RCIM_Cubic>,<at 0.2126 0.0616 RCIM_Cubic>]}
      }})
    R1 Saffron:LinearColor=&ColorFromCurve(ColorCurve=DataInterface() {propertyValues {
       AlphaCurve {keys [<at 0 1 RCIM_Cubic>,<at 1 0 RCIM_Cubic>]}
       BlueCurve {keys [<at 0 0.1051 RCIM_Cubic>,<at 0.2126 0 RCIM_Cubic>]}
       GreenCurve {keys [<at 0 0.252 RCIM_Cubic>,<at 0.2126 0.5227 RCIM_Cubic>]}
       RedCurve {keys [<at 0 0.2969 RCIM_Cubic>,<at 0.2126 0.6823 RCIM_Cubic>]}
      }})
    R1 SkyBlue:LinearColor=&ColorFromCurve(ColorCurve=DataInterface() {propertyValues {
       AlphaCurve {keys [<at 0 1 RCIM_Cubic>,<at 1 0 RCIM_Cubic>]}
       BlueCurve {keys [<at 0 0.4323 RCIM_Cubic>,<at 0.2126 0.3698 RCIM_Cubic>]}
       GreenCurve {keys [<at 0 0.3419 RCIM_Cubic>,<at 0.2126 0.2471 RCIM_Cubic>]}
       RedCurve {keys [<at 0 0.1599 RCIM_Cubic>,<at 0.2126 0 RCIM_Cubic>]}
      }})
    R1 TitaniumWhite:LinearColor=&ColorFromCurve(ColorCurve=DataInterface() {propertyValues {
       AlphaCurve {keys [<at 0 0.9 RCIM_Cubic>,<at 1 0 RCIM_Cubic>]}
       BlueCurve {keys [<at 0 0.3712 RCIM_Cubic>,<at 0.2943 0.8125 RCIM_Cubic>]}
       GreenCurve {keys [<at 0 0.3712 RCIM_Cubic>,<at 0.2943 0.8125 RCIM_Cubic>]}
       RedCurve {keys [<at 0 0.3712 RCIM_Cubic>,<at 0.2943 0.8125 RCIM_Cubic>]}
      }})
    R2 Black:LinearColor=<rgba 0.1094 0.1094 0.1094 1>
    R2 BurntSienna:LinearColor=<rgba 0.4219 0.1303 0.052 1>
    R2 Cobalt:LinearColor=<rgba 0.0625 0.0798 1 1>
    R2 Crimson:LinearColor=<rgba 0.3802 0.0356 0.0448 1>
    R2 Grey:LinearColor=<rgba 0.0417 0.0417 0.0417 1>
    R2 Purple:LinearColor=<rgba 0.0844 0.0482 0.3372 1>
    Saffron=<rgba 0.2708 0.2076 0 1>
    SkyBlue=<rgba 0 0.357 0.5365 1>
    TitaniumWhite=<rgba 1.5 1.5 1.5 1>
    Use R1 Painted Overrides=true
    Use R2 Painted Overrides=true
   ) {module=VC_PaintedVariantsOverride001}
   &ScaleRibbonWidth(Ribbon Width Scale=&FloatFromCurve(FloatCurve=DataInterface() {propertyValues {Curve {keys [
         <at 0 0.4 RCIM_Cubic>
         {arriveTangentWeight=0.0731, interpMode=RCIM_Cubic, leaveTangentWeight=0.2924, tangentMode=RCTM_User, time=0.2, value=0.5}
         <at 1 0.2 RCIM_Cubic>
        ]}}})) {enabled=false}
   &InheritSourceMovement(
    Applied Position Scale=&VectorFromFloat(Value=&FloatFromCurve(FloatCurve=DataInterface() {propertyValues {Curve {keys [<at 0 1 RCIM_Cubic>,<at 1 0 RCIM_Cubic>]}}}))
    ApplyForce=false
    ApplyPosition=true
    ApplyPositionFalloff=false
   )
   &SolveForcesAndVelocity(Clamp Velocity=false, Limit Acceleration=false)
   &VC_UpdateRibbon_v1(Enable Skeletal Position=true)
   &DynamicMaterialParameters(
    Index 0 Param 1=2
    Index 0 Param 2=0.5
    Index 0 Param 3=1
    Index 0 Param 4=1
    Module.Index0Float0Write=true
    Module.Index0Float1Write=true
    Module.Index0Float2Write=true
    Module.Index0Float3Write=true
   )
   /SetVariables(
    Particles.Index1:Vector4f=&MakeVector4(X=3, Y=3, Z=1, W=-1)
    Particles.Index2:Vector4f=&MakeVector4(X=2, Y=1, Z=0, W=0)
    Particles.Index3:Vector4f=&MakeVector4(X=2, Y=&MakeCustomFloatFromBool(Bool=$User.bIsPainted, True Float=300, False Float=100), Z=8, W=-1)
    Particles.RibbonColor0:LinearColor=&VC_DynamicScaleColor(
     Color to Scale:LinearColor=$Particles.Color
     RGB Scale:Vector3f=&MakeCustomVectorFromBool(
      Bool=$User.bIsPainted
      True Vector=&VectorFromFloat(Value=0.1)
      False Vector=&MakeVectorFromLinearColorRGB(Color=<rgba 0.08 0.05 0.1 1>)
     )
     Alpha Scale:NiagaraFloat=&FloatFromCurve(FloatCurve=DataInterface() {propertyValues {Curve {keys [
          <at 0 0.5 RCIM_Cubic>
          {arriveTangentWeight=0.1572, interpMode=RCIM_Cubic, leaveTangentWeight=0.1572, tangentMode=RCTM_User, time=0.3, value=1}
          <at 0.7 0 RCIM_Cubic>
         ]}}})
    )
    Particles.RibbonColor1:LinearColor=&VC_DynamicScaleColor(
     Color to Scale:LinearColor=&MakeCustomLinearColorFromBool(
      Bool=$User.bIsPainted
      True Color=$Particles.R1 Color
      False Color=&ColorFromCurve(ColorCurve=DataInterface() {propertyValues {
         AlphaCurve {keys [<at 0 1 RCIM_Cubic>,<at 1 1 RCIM_Cubic>]}
         BlueCurve {keys [
           <at 0.0459 0.1458 RCIM_Cubic>
           {arriveTangentWeight=0.1346, interpMode=RCIM_Cubic, leaveTangentWeight=0.1565, tangentMode=RCTM_User, time=0.1, value=1}
           <at 0.25 1 RCIM_Cubic>
          ]}
         GreenCurve {keys [
           <at 0.0459 1 RCIM_Cubic>
           {arriveTangent=-4.5941, interpMode=RCIM_Cubic, leaveTangent=-4.5941, time=0.1, value=0.3005}
           <at 0.25 0.0623 RCIM_Cubic>
          ]}
         RedCurve {keys [
           <at 0.0459 1 RCIM_Cubic>
           {arriveTangentWeight=0.1339, interpMode=RCIM_Cubic, leaveTangentWeight=0.1557, tangentMode=RCTM_User, time=0.1, value=1}
           <at 0.25 0.1503 RCIM_Cubic>
          ]}
        }})
     )
     RGB Scale:Vector3f=&MakeCustomVectorFromBool(
      Bool=$User.bIsPainted
      True Vector=&VectorFromFloat(Value=&FloatFromCurve(FloatCurve=DataInterface() {propertyValues {Curve {keys [<at 0 5 RCIM_Cubic>,<at 0.3 0.5 RCIM_Cubic>]}}}))
      False Vector=&VectorFromFloat(Value=&FloatFromCurve(FloatCurve=DataInterface() {propertyValues {Curve {keys [<at 0 2 RCIM_Cubic>,<at 0.3 0.5 RCIM_Cubic>]}}}))
     )
     Alpha Scale:NiagaraFloat=&FloatFromCurve(FloatCurve=DataInterface() {propertyValues {Curve {keys [
          <at 0 0 RCIM_Cubic>
          {arriveTangentWeight=0.2247, interpMode=RCIM_Cubic, leaveTangentWeight=0.4494, tangentMode=RCTM_User, time=0.1, value=2}
          {arriveTangentWeight=0.0667, interpMode=RCIM_Cubic, leaveTangentWeight=0.0667, tangentMode=RCTM_User, time=0.3, value=2}
          {arriveTangentWeight=0.7723, interpMode=RCIM_Cubic, tangentMode=RCTM_User, time=0.5, value=0}
         ]}}})
    )
    Particles.RibbonColor2:LinearColor=&VC_DynamicScaleColor(
     Color to Scale:LinearColor=&MakeCustomLinearColorFromBool(
      Bool=$User.bIsPainted
      True Color=$Particles.R1 Color
      False Color=&ColorFromCurve(ColorCurve=DataInterface() {propertyValues {
         AlphaCurve {keys [<at 0 1 RCIM_Cubic>,<at 1 1 RCIM_Cubic>]}
         BlueCurve {keys [<at 0.1234 0.4 RCIM_Cubic>,<at 0.1866 1>,<at 0.2559 1 RCIM_Cubic>]}
         GreenCurve {keys [<at 0.1234 1 RCIM_Cubic>,<at 0.1866 0.3>,<at 0.2559 0.1 RCIM_Cubic>]}
         RedCurve {keys [<at 0.1234 1 RCIM_Cubic>,<at 0.1866 1>,<at 0.2559 0.2 RCIM_Cubic>]}
        }})
     )
     RGB Scale:Vector3f=&VectorFromFloat(Value=&FloatFromCurve(FloatCurve=DataInterface() {propertyValues {Curve {keys [<at 0.2 1>,<at 0.5 2>]}}}, Scale Curve=1.5))
     Alpha Scale:NiagaraFloat=&FloatFromCurve(
      FloatCurve=DataInterface() {propertyValues {Curve {keys [
          <at 0 0 RCIM_Cubic>
          {interpMode=RCIM_Cubic, leaveTangentWeight=0.0833, tangentMode=RCTM_User, time=0.2, value=1}
          {arriveTangentWeight=0.1008, interpMode=RCIM_Cubic, tangentMode=RCTM_User, time=0.5, value=0}
         ]}}}
      Scale Curve=4
     )
    )
    Particles.RibbonColor3:LinearColor=&VC_DynamicScaleColor(
     Color to Scale:LinearColor=&MakeCustomLinearColorFromBool(
      Bool=$User.bIsPainted
      True Color=$Particles.R1 Color
      False Color=&ColorFromCurve(ColorCurve=DataInterface() {propertyValues {
         AlphaCurve {keys [<at 0 1 RCIM_Cubic>,<at 1 1 RCIM_Cubic>]}
         BlueCurve {keys [<at 0.0459 0.1458 RCIM_Cubic>,<at 0.1295 1>,<at 0.2417 0.8542>]}
         GreenCurve {keys [<at 0.0459 0.9372 RCIM_Cubic>,<at 0.1295 0.3005>,<at 0.2417 0.0623>]}
         RedCurve {keys [<at 0.0459 1 RCIM_Cubic>,<at 0.1295 1>,<at 0.2417 0.1503>]}
        }})
     )
     RGB Scale:Vector3f=&MakeCustomVectorFromBool(
      Bool=$User.bIsPainted
      True Vector=&VectorFromFloat(Value=&FloatFromCurve(FloatCurve=DataInterface() {propertyValues {Curve {keys [<at 0 50>,<at 0.3 0.5>]}}}))
      False Vector=&VectorFromFloat(Value=1)
     )
     Alpha Scale:NiagaraFloat=&FloatFromCurve(
      FloatCurve=DataInterface() {propertyValues {Curve {keys [
          {arriveTangent=11.5571, interpMode=RCIM_Cubic, leaveTangent=11.5571, tangentMode=RCTM_User, time=0, value=0}
          <at 0.25 1 RCIM_Cubic>
          {arriveTangent=-9.2133, interpMode=RCIM_Cubic, leaveTangent=-9.2133, tangentMode=RCTM_User, time=0.5, value=0}
         ]}}}
      Scale Curve=0.5
     )
    )
    Particles.RibbonWidth0:NiagaraFloat=&Multiply_Float(
     A=&FloatFromCurve(FloatCurve=DataInterface() {propertyValues {Curve {keys [<at 0 1.5>,<at 1 0.35>]}}}, Scale Curve=1.6)
     B=$Particles.RibbonWidth
    )
    Particles.RibbonWidth1:NiagaraFloat=&Multiply_Float(
     A=&FloatFromCurve(FloatCurve=DataInterface() {propertyValues {Curve {keys [
          {arriveTangent=-1.5, interpMode=RCIM_Cubic, time=0, value=0.4}
          {arriveTangentWeight=0.1221, interpMode=RCIM_Cubic, leaveTangentWeight=0.2848, tangentMode=RCTM_User, time=0.2, value=1}
          <at 0.5 0.2 RCIM_Cubic>
         ]}}})
     B=$Particles.RibbonWidth
    )
    Particles.RibbonWidth2:NiagaraFloat=&Multiply_Float(A=1, B=$Particles.RibbonWidth)
    Particles.RibbonWidth3:NiagaraFloat=&Multiply_Float(
     A=&FloatFromCurve(FloatCurve=DataInterface() {propertyValues {Curve {keys [<at 0 1>,<at 1 0>]}}})
     B=$Particles.RibbonWidth
    )
   ) {isSetParameters=true}
  ]
  rendererClasses [&NiagaraRibbonRendererProperties]
  renderers [
   {rendererClass=&NiagaraRibbonRendererProperties}
   {rendererClass=&NiagaraRibbonRendererProperties, rendererIndex=1}
   {rendererClass=&NiagaraRibbonRendererProperties, rendererIndex=2}
   {rendererClass=&NiagaraRibbonRendererProperties, rendererIndex=3}
  ]
 }
 NE_VC_Boost_Ribbon_Parent003 {
  emitterUpdateScript [
   &EmitterState()
   &VC_FilteredSocketCount()
   &VC_RateQualityHandler()
   &VC_SpawnRibbon_v1(Number of Ribbons=2, SpawnRate=60)
  ]
  particleSpawnScript [
   &VC_PaintedVariantsOverride(
    Black=<rgba 0.05 0.05 0.05 1>
    BurntSienna=<rgba 0.15 0.0375 0 1>
    Cobalt=<rgba 0.18 0.288 1.8 1>
    Crimson=<rgba 1.3 0.13 0.1617 1>
    DefaultColor=<rgba 1 0.7799 0 1>
    ForestGreen=<rgba 0.2 1 0.2404 1>
    Grey=<rgba 0.28 0.28 0.28 1>
    Lime=<rgba 0.65 1 0.3 1>
    Orange=<rgba 1.2 0.5 0 1>
    Pink=<rgba 1 0.25 0.7273 1>
    Purple=<rgba 0.4 0.2 0.8 1>
    R1 Black:LinearColor=&Multiply_LinearColorByFloat(Linear Color=<rgba 0 0 0 1>, Float=3)
    R1 BurntSienna:LinearColor=<rgba 0.013 0.004 0.0016 3>
    R1 Cobalt:LinearColor=<rgba 0 0.0185 1 0.5>
    R1 Crimson:LinearColor=<rgba 0.464 0.1392 0.148 0.5>
    R1 DefaultColor=&Multiply_LinearColorByFloat(Linear Color=<rgba 1 1 1 1>, Float=2)
    R1 ForestGreen:LinearColor=<rgba 0 0.5 0.0253 0.5>
    R1 Grey:LinearColor=<rgba 0.1 0.1 0.1 2>
    R1 Lime:LinearColor=<rgba 0.2943 0.5885 0 0.2>
    R1 Orange:LinearColor=<rgba 0.6 0.15 0 0.8>
    R1 Pink:LinearColor=<rgba 1 0.3 0.7455 0.2>
    R1 Purple:LinearColor=<rgba 0.0787 0.0213 0.71 1.3>
    R1 Saffron:LinearColor=<rgba 0.7 0.5833 0 0.5>
    R1 SkyBlue:LinearColor=<rgba 0 0.357 0.5365 0.25>
    R1 TitaniumWhite:LinearColor=<rgba 0.5 0.5 0.5 0.35>
    R2 Black:LinearColor=&Multiply_LinearColorByFloat(Linear Color=<rgba 0 0 0 1>, Float=3)
    R2 BurntSienna:LinearColor=<rgba 0.03 0.0101 0.006 1>
    R2 Cobalt:LinearColor=<rgba 0 0.025 1.5 0.5>
    R2 Crimson:LinearColor=<rgba 0.6 0.03 0.049 0.7>
    R2 DefaultColor:LinearColor=<rgba 0 2.8619 4 0.2261>
    R2 ForestGreen:LinearColor=<rgba 0 0.6 0.0303 1>
    R2 Grey:LinearColor=<rgba 0.14 0.14 0.14 0.8>
    R2 Lime:LinearColor=<rgba 0.315 0.6 0.03 1>
    R2 Orange:LinearColor=<rgba 0.9 0.165 0 0.8>
    R2 Pink:LinearColor=<rgba 0.8 0.096 0.544 0.8>
    R2 Purple:LinearColor=<rgba 0.0875 0.05 0.5 0.4>
    R2 Saffron:LinearColor=<rgba 1 0.8333 0 0.75>
    R2 SkyBlue:LinearColor=<rgba 0 0.5 1 0.7>
    R2 TitaniumWhite:LinearColor=<rgba 2 2 2 0.25>
    Saffron=<rgba 1.2 1.004 0.36 1>
    SkyBlue=<rgba 0.25 0.749 1 1>
    TitaniumWhite=<rgba 0.3 0.3 0.3 1>
    Use R1 Painted Overrides=true
    Use R2 Painted Overrides=true
   )
   &VC_InitializeRibbon_v1(
    Color=$Particles.Color
    Custom Offset=true
    Enable Skeletal Position=true
    Lifetime=0.6
    Local Vector Offset=&AddVectorToPosition(Position=<-70 0 0>, Vector=Linked() {linkedVariable {name=User.DTCarOffset, type=NiagaraPosition}})
    Offset Amount=1
    Ribbon Width=60
    Write Material Random=false
   )
   &AddVelocity(Velocity=<-2000 0 0>) {module=AddVelocity001}
  ]
  particleUpdateScript [
   &ParticleState()
   &InheritSourceMovement(ApplyForce=false, ApplyPosition=true, ApplyPositionFalloff=false)
   &SolveForcesAndVelocity(Clamp Velocity=false, Limit Acceleration=false)
   &VC_UpdateRibbon_v1(Enable Skeletal Position=true)
   &ScaleColor(
    Scale Alpha=&FloatFromCurve(
     FloatCurve=DataInterface() {propertyValues {Curve {
        defaultValue=1
        keys [
         {arriveTangent=3.8118, interpMode=RCIM_Cubic, leaveTangent=3.8118, tangentMode=RCTM_User, time=0, value=0}
         {arriveTangent=1.5427, interpMode=RCIM_Cubic, leaveTangent=-0.033, tangentMode=RCTM_Break, time=0.1789, value=0.9847}
         {arriveTangent=-5.5211, interpMode=RCIM_Cubic, leaveTangent=-5.5211, tangentMode=RCTM_User, time=0.8734, value=0}
        ]
       }}}
     Scale Curve=20
    )
    Scale RGB=&VectorFromFloat(Value=1)
    ScaleA=true
    ScaleRGB=true
   )
   /SetVariables(
    Particles.BeamColor:LinearColor=&MultiplyLinearColor(
     Linear Color=<rgba 1 0.9581 0 1>
     Scale Factor=&MakeVector4FromVectorAndFloat(
      "Vector (RGB)"=<1 1 1>
      "Float (Alpha)"=&FloatFromCurve(FloatCurve=DataInterface() {propertyValues {Curve {
          defaultValue=1
          keys [
           {arriveTangent=1.4927, interpMode=RCIM_Cubic, time=-0.0449, value=0.734}
           {arriveTangent=1.5427, interpMode=RCIM_Cubic, leaveTangent=-1.0354, tangentMode=RCTM_Break, time=0.1789, value=0.9847}
           {arriveTangent=-3.3735, interpMode=RCIM_Cubic, leaveTangent=-3.3735, tangentMode=RCTM_User, time=1.0198, value=-0.2103}
          ]
         }}})
     )
    )
    Particles.BlueWidth:NiagaraFloat=&FloatFromCurve(
     FloatCurve=DataInterface() {propertyValues {Curve {
        defaultValue=1
        keys [
         <at -0.0085 0.7008 RCIM_Cubic>
         {arriveTangent=0.0448, interpMode=RCIM_Cubic, leaveTangent=0.0448, tangentMode=RCTM_User, time=0.2772, value=0.9847}
         {arriveTangent=-0.932, interpMode=RCIM_Cubic, leaveTangent=-0.932, tangentMode=RCTM_User, time=1.0722, value=0.538}
        ]
       }}}
     Scale Curve=150
    )
    Particles.CrossColor:LinearColor=&MakeCustomLinearColorFromBool(
     Bool=$User.bIsPainted
     True Color=&MultiplyLinearColor(
      Linear Color=<rgba 1 0.7986 0 1>
      Scale Factor=&MakeVector4FromVectorAndFloat(
       "Vector (RGB)"=<1 1 1>
       "Float (Alpha)"=&FloatFromCurve(
        FloatCurve=DataInterface() {propertyValues {Curve {
           defaultValue=1
           keys [
            {arriveTangent=1.4927, interpMode=RCIM_Cubic, time=0, value=0}
            {arriveTangent=0.1055, interpMode=RCIM_Cubic, leaveTangent=-1.0354, tangentMode=RCTM_Break, time=0.2653, value=1}
            {arriveTangent=-0.2248, interpMode=RCIM_Cubic, leaveTangent=-0.2248, tangentMode=RCTM_User, time=0.9436, value=0}
           ]
          }}}
        Scale Curve=3
       )
      )
     )
     False Color=&ColorFromCurve(ColorCurve=DataInterface() {propertyValues {
        AlphaCurve {keys [<at 0 0 RCIM_Cubic>,<at 0.4186 1>,<at 1 2 RCIM_Cubic>]}
        BlueCurve {keys [<at 0 1 RCIM_Cubic>,<at 0.7607 1 RCIM_Cubic>]}
        GreenCurve {keys [<at 0 0.6654 RCIM_Cubic>,<at 0.7607 0 RCIM_Cubic>]}
        RedCurve {keys [<at 0 0 RCIM_Cubic>,<at 0.7607 0.029 RCIM_Cubic>]}
       }})
    )
    Particles.CrossColor2:LinearColor=&MakeCustomLinearColorFromBool(
     Bool=$User.bIsFrontendPreview
     True Color=&MultiplyLinearColor(
      Linear Color=<rgba 1 1 1 1>
      Scale Factor=&MakeVector4FromVectorAndFloat(
       "Vector (RGB)"=<1 1 1>
       "Float (Alpha)"=&FloatFromCurve(FloatCurve=DataInterface() {propertyValues {Curve {
           defaultValue=1
           keys [
            {arriveTangent=1.4927, interpMode=RCIM_Cubic, time=-0.0449, value=0.734}
            {arriveTangent=1.5427, interpMode=RCIM_Cubic, leaveTangent=-1.0354, tangentMode=RCTM_Break, time=0.1789, value=0.9847}
            {arriveTangent=-3.3735, interpMode=RCIM_Cubic, leaveTangent=-3.3735, tangentMode=RCTM_User, time=1.0198, value=-0.2103}
           ]
          }}})
      )
     )
     False Color=&MultiplyLinearColor(
      Linear Color=$Particles.R1 Color
      Scale Factor=&MakeVector4FromVectorAndFloat(
       "Vector (RGB)"=<1 5 2>
       "Float (Alpha)"=&FloatFromCurve(
        FloatCurve=DataInterface() {propertyValues {Curve {
           defaultValue=1
           keys [
            {arriveTangent=1.4927, interpMode=RCIM_Cubic, time=-0.0449, value=0.734}
            {arriveTangent=1.5427, interpMode=RCIM_Cubic, leaveTangent=-1.0354, tangentMode=RCTM_Break, time=0.1789, value=0.9847}
            {arriveTangent=-3.3735, interpMode=RCIM_Cubic, leaveTangent=-3.3735, tangentMode=RCTM_User, time=1.0198, value=-0.2103}
           ]
          }}}
        Scale Curve=0.5
       )
      )
     )
    )
    Particles.CrossInvertWidth:NiagaraFloat=&FloatFromCurve(
     FloatCurve=DataInterface() {propertyValues {Curve {
        defaultValue=1
        keys [
         <at -0.0085 0.7008 RCIM_Cubic>
         {arriveTangent=0.0448, interpMode=RCIM_Cubic, leaveTangent=0.0448, tangentMode=RCTM_User, time=0.2772, value=0.9847}
         {arriveTangent=-0.0405, interpMode=RCIM_Cubic, leaveTangent=-0.0405, tangentMode=RCTM_User, time=0.8385, value=0.0052}
        ]
       }}}
     Scale Curve=10
    )
    Particles.CrossWidth:NiagaraFloat=&FloatFromCurve(
     FloatCurve=DataInterface() {propertyValues {Curve {
        defaultValue=1
        keys [
         <at -0.0085 0.7008 RCIM_Cubic>
         {arriveTangent=0.0448, interpMode=RCIM_Cubic, leaveTangent=0.0448, tangentMode=RCTM_User, time=0.2772, value=0.9847}
         {arriveTangent=-0.3044, interpMode=RCIM_Cubic, leaveTangent=-0.3044, tangentMode=RCTM_User, time=0.8, value=0}
        ]
       }}}
     Scale Curve=50
    )
    Particles.LightningColor:LinearColor=&MultiplyLinearColor(
     Linear Color=<rgba 1 0.8442 0 1>
     Scale Factor=&MakeVector4FromVectorAndFloat(
      "Vector (RGB)"=&VectorFromFloat(Value=5)
      "Float (Alpha)"=&FloatFromCurve(FloatCurve=DataInterface() {propertyValues {Curve {
          defaultValue=1
          keys [
           {arriveTangent=0.0901, interpMode=RCIM_Cubic, leaveTangent=0.0901, tangentMode=RCTM_User, time=-0.0449, value=-0.0307}
           {arriveTangent=-0.4524, interpMode=RCIM_Cubic, leaveTangent=0.0385, tangentMode=RCTM_Break, time=0.5314, value=0.9847}
           {arriveTangent=-0.1302, interpMode=RCIM_Cubic, leaveTangent=-0.1302, tangentMode=RCTM_User, time=1.4722, value=0}
          ]
         }}})
     )
    )
    Particles.ThunderWidth:NiagaraFloat=&FloatFromCurve(
     FloatCurve=DataInterface() {propertyValues {Curve {
        defaultValue=1
        keys [
         {arriveTangent=1.4927, interpMode=RCIM_Cubic, time=-0.0449, value=0.734}
         {arriveTangent=1.5427, interpMode=RCIM_Cubic, leaveTangent=-0.1679, tangentMode=RCTM_Break, time=0.1789, value=0.9847}
         {arriveTangent=-2.714, interpMode=RCIM_Cubic, leaveTangent=-2.714, tangentMode=RCTM_User, time=0.95, value=0}
        ]
       }}}
     Scale Curve=50
    )
   ) {isSetParameters=true, module=SetVariables_#2}
   /SetVariables(
    Particles.CrossInvertPosition:NiagaraPosition=&AddVectorToPosition(
     Position=$Particles.Position
     Vector=&TransformVector(
      Vector=&MakeVector(
       X=&SelectFloatFromArray(Direct Array Index=&SpawnGroup(), Float Selection Array=DataInterface() {propertyValues {FloatData [-30,0]}})
       Y=0
       Z=&Sine(
        Normalized Angle=&Add_Float(
         A=&Multiply_Float(A=$Particles.NormalizedAge, B=-1.3)
         B=&Multiply_Float(A=$Emitter.LoopedAge, B=1.2)
        )
        Period=-1
        Scale=&FloatFromCurve(
         FloatCurve=DataInterface() {propertyValues {Curve {
            defaultValue=1
            keys [
             {arriveTangent=2.952, interpMode=RCIM_Cubic, leaveTangent=2.952, tangentMode=RCTM_User, time=-0.0185, value=0}
             {arriveTangent=0.1401, tangentMode=RCTM_User, time=0.387, value=0.6595}
             {arriveTangent=-1.5773, interpMode=RCIM_Cubic, leaveTangent=-3.8709, tangentMode=RCTM_Break, time=0.5819, value=0.6595}
             {arriveTangent=-0.2749, interpMode=RCIM_Cubic, leaveTangent=-0.2749, tangentMode=RCTM_User, time=1.0012, value=0}
            ]
           }}}
         Scale Curve=20
        )
        Bias=0
       )
      )
      Destination Space=::Simulation
      Apply Scale=true
     )
    )
    Particles.PositionCross:NiagaraPosition=&AddVectorToPosition(
     Position=$Particles.Position
     Vector=&TransformVector(
      Vector=&MakeVector(
       X=&SelectFloatFromArray(Direct Array Index=&SpawnGroup(), Float Selection Array=DataInterface() {propertyValues {FloatData [-30,0]}})
       Y=0
       Z=&Sine(
        Normalized Angle=&Add_Float(
         A=&Multiply_Float(A=$Particles.NormalizedAge, B=-1.3)
         B=&Multiply_Float(A=$Emitter.LoopedAge, B=1.5)
        )
        Period=1
        Scale=&FloatFromCurve(
         FloatCurve=DataInterface() {propertyValues {Curve {
            defaultValue=1
            keys [
             {arriveTangent=-0.1097, interpMode=RCIM_Cubic, leaveTangent=-0.1097, tangentMode=RCTM_User, time=0.175, value=0}
             {arriveTangent=7.2017, leaveTangent=7.2017, time=0.3137, value=0.7182}
             {arriveTangent=-2.5934, interpMode=RCIM_Cubic, leaveTangent=-2.5934, tangentMode=RCTM_User, time=0.4951, value=0.7335}
             {arriveTangent=-0.0405, interpMode=RCIM_Cubic, leaveTangent=-0.0405, tangentMode=RCTM_User, time=0.7, value=0}
            ]
           }}}
         Scale Curve=15
        )
        Bias=0
       )
      )
      Destination Space=::Simulation
      Apply Scale=true
     )
    )
    Particles.PositionLightning:NiagaraPosition=&AddVectorToPosition(
     Position=$Particles.Position
     Vector=&TransformVector(
      Vector=&MakeVector(
       X=0
       Y=0
       Z=&Sine(
        Normalized Angle=&Add_Float(
         A=&Multiply_Float(A=$Particles.NormalizedAge, B=-1.3)
         B=&Multiply_Float(A=$Emitter.LoopedAge, B=1.3)
        )
        Period=1
        Scale=&FloatFromCurve(
         FloatCurve=DataInterface() {propertyValues {Curve {
            defaultValue=1
            keys [
             {arriveTangent=9.0725, interpMode=RCIM_Cubic, leaveTangent=9.0725, tangentMode=RCTM_User, time=-0.0048, value=0}
             {arriveTangent=0.0244, interpMode=RCIM_Cubic, leaveTangent=0.0244, tangentMode=RCTM_User, time=0.1408, value=0.9847}
             {arriveTangent=-0.0583, interpMode=RCIM_Cubic, leaveTangent=-0.0583, tangentMode=RCTM_User, time=0.3728, value=1}
             {arriveTangent=-0.2024, interpMode=RCIM_Cubic, leaveTangent=-0.2024, tangentMode=RCTM_User, time=1.1034, value=0}
            ]
           }}}
         Scale Curve=13
        )
        Bias=0
       )
      )
      Destination Space=::Simulation
      Apply Scale=true
     )
    )
   ) {isSetParameters=true, module=SetVariables_#3}
  ]
  rendererClasses [&NiagaraRibbonRendererProperties]
  renderers [
   {rendererClass=&NiagaraRibbonRendererProperties}
   {rendererClass=&NiagaraRibbonRendererProperties, rendererIndex=1}
   {rendererClass=&NiagaraRibbonRendererProperties, rendererIndex=2}
   {rendererClass=&NiagaraRibbonRendererProperties, rendererIndex=3}
  ]
 }
 NE_VC_DriftTrail_WheelWrap_R001 {
  bEnabled=false
  emitterUpdateScript [
   &EmitterState(
    Inactive Response=::NewEnumerator1
    Life Cycle Mode=::NewEnumerator1
    Loop Behavior=::NewEnumerator1
    Loop Duration Mode=::NewEnumerator1
    UseLoopDelay=false
   )
   &SpawnBurst_Instantaneous(Spawn Count=4, Use Spawn Probability=false, UseLoopCountLimit=false)
  ]
  particleSpawnScript [
   &VC_PaintedVariantsOverride() {enabled=false}
   &InitializeParticle(
    Color=&MakeCustomLinearColorFromBool(Bool=$User.bIsPainted, True Color=$User.PaintedColor, False Color=<rgba 1 1 1 1>)
    Mesh Scale Mode=::NewEnumerator4
    Sprite Rotation Mode=::NewEnumerator2
    Sprite Size Mode=::NewEnumerator3
    UsePositionOffset=false
    Write Material Random=true
   )
   &PartitionParticles()
   &SkeletalMeshLocation(
    Bone / Socket Index=&Add_Integer(A=$Particles.PartitionParticles.Partition, B=4)
    Bone / Socket Sampling Mode=::NewEnumerator5
    Interpolate Mesh Sampling=false
    MaskSpawnGroup=false
    Mesh Sampling Type=::NewEnumerator6
    Orientation Sampling=::NewEnumerator1
    Sampled Position Offset Coordinate Space:ENiagaraSkelMeshTransforms=::NewEnumerator2
    Skeletal Mesh=$System.SkeletalMesh
   )
  ]
  particleUpdateScript [
   &FN_DisableTireFX(
    Mask Index Attribute:NiagaraInt32=$Particles.PartitionParticles.Partition
    Mask Index FL=2
    Mask Index FR=3
    Mask Index RL=0
    Mask Index RR=1
   )
   &VC_WheelWrapScale(Wheel Index Attribute:NiagaraInt32=&Add_Integer(A=$Particles.PartitionParticles.Partition, B=2))
   &ParticleState(Kill Particles When Lifetime Has Elapsed=false, Loop Particles Lifetime=true)
   &UpdateMeshOrientation(Rotation Coordinate Space=::NewEnumerator2, Rotation Vector=<0 1 0>) {enabled=false}
   &SkeletalMeshLocation(
    Bone / Socket Index=&Add_Integer(A=$Particles.PartitionParticles.Partition, B=4)
    Bone / Socket Sampling Mode=::NewEnumerator5
    MaskSpawnGroup=false
    Mesh Sampling Type=::NewEnumerator6
    Skeletal Mesh=$System.SkeletalMesh
   ) {module=SkeletalMeshLocation001}
   &SolveForcesAndVelocity(Clamp Velocity=false, Limit Acceleration=false)
  ]
  rendererClasses [&NiagaraMeshRendererProperties]
  renderers [{rendererClass=&NiagaraMeshRendererProperties, rendererIndex=0}]
 }
 NE_VC_Ribbon_Parent_Trail4 {
  emitterUpdateScript [
   &EmitterState(Life Cycle Mode=::NewEnumerator1, Loop Duration=1, UseLoopDelay=false)
   &VC_FilteredSocketCount()
   &VC_SpawnRibbon_v1(Number of Ribbons=2, SpawnRate=60)
  ]
  particleSpawnScript [
   &VC_PaintedVariantsOverride() {enabled=false}
   &VC_InitializeRibbon_v1(
    Color=&MakeCustomLinearColorFromBool(Bool=$User.bIsPainted, True Color=$User.PaintedColor, False Color=<rgba 0.0575 0.1352 0.1406 1>)
    Enable Skeletal Position=true
    Lifetime=0.8
    Ribbon Width=150
   )
   &AddVelocity(
    Velocity=&MakeCustomVectorFromBool(Bool=$User.bIsFrontendPreview, True Vector=<1500 0 0>, False Vector=<1500 0 0>)
    Velocity Speed Scale=-1
   ) {module=AddVelocity001}
  ]
  particleUpdateScript [
   &ParticleState()
   &ScaleColor(
    Scale Alpha=&MakeCustomFloatFromBool(
     Bool=$User.bIsPainted
     True Float=1
     False Float=&FloatFromCurve(FloatCurve=DataInterface() {propertyValues {Curve {keys [<at -0.0469 3 RCIM_Cubic>,<at 1 0 RCIM_Cubic>]}}})
    )
    Scale RGB=&MakeCustomVectorFromBool(Bool=$User.bIsPainted, True Vector=<1 1 1>, False Vector=<1 1 1>)
    ScaleA=true
    ScaleRGB=true
   )
   &SolveForcesAndVelocity(Clamp Velocity=false, Limit Acceleration=false)
   &VC_UpdateRibbon_v1(Enable Skeletal Position=true)
   &DynamicMaterialParameters(
    Index 0 Param 1=1.7
    Index 0 Param 3=1
    Module.Index0Param0Write=true
    Module.Index0Param1Write=true
    Module.Index0Param2Write=true
    Module.Index0Param3Write=true
   )
  ]
  rendererClasses [&NiagaraRibbonRendererProperties]
  renderers [{rendererClass=&NiagaraRibbonRendererProperties}]
 }
 Red_Blue {
  emitterUpdateScript [
   &EmitterState()
   &VC_FilteredSocketCount()
   &VC_SpawnRibbon_v1(Number of Ribbons=2, SpawnRate=60)
  ]
  particleSpawnScript [
   &VC_PaintedVariantsOverride() {enabled=false}
   &VC_InitializeRibbon_v1(
    Color=<rgba 0.0275 0.0157 0.0941 1>
    Custom Offset=true
    Enable Skeletal Position=true
    Lifetime=1.6
    Local Vector Offset=&AddVectorToPosition(Position=<0 0 0>, Vector=$User.DTCarOffset)
    Offset Amount=1
    Ribbon Width:NiagaraFloat=&Multiply_Float(A=4, B=5.5)
   )
   &AddVelocity(Velocity=<-1000 0 0>) {module=AddVelocity001}
  ]
  particleUpdateScript [
   &ParticleState()
   &InheritSourceMovement(
    Applied Position Scale=&Multiply_VectorByFloat(
     Vector=<1 1 1>
     Float=&FloatFromCurve(FloatCurve=DataInterface() {propertyValues {Curve {keys [
          {interpMode=RCIM_Cubic, leaveTangent=-0.0051, tangentMode=RCTM_Break, time=0, value=1}
          {arriveTangent=-0.247, interpMode=RCIM_Cubic, tangentMode=RCTM_Break, time=1, value=0.9}
         ]}}})
    )
    ApplyForce=false
    ApplyPosition=true
    ApplyPositionFalloff=false
   )
   &SolveForcesAndVelocity(Clamp Velocity=false, Limit Acceleration=false)
   &VC_UpdateRibbon_v1(Enable Skeletal Position=true)
   &ScaleColor(
    Scale Alpha=&FloatFromCurve(FloatCurve=DataInterface() {propertyValues {Curve {keys [
         {interpMode=RCIM_Cubic, leaveTangent=0.7556, tangentMode=RCTM_Break, time=0, value=0}
         {interpMode=RCIM_Cubic, tangentMode=RCTM_Break, time=0.5, value=1}
         {interpMode=RCIM_Cubic, tangentMode=RCTM_Break, time=1, value=-0.2}
        ]}}})
    Scale RGB=<2.5 2.5 2.5>
    ScaleA=true
    ScaleRGB=true
   )
   &DynamicMaterialParameters(
    Index 0 Param 1=&SelectFloatFromArray(Direct Array Index=$Particles.SpawnGroup, Float Selection Array=DataInterface() {propertyValues {FloatData [0,0.5]}})
    Index 0 Param 2=-2
    Index 0 Param 3=1
    Index 0 Param 4=4
    Module.Index0Param0Write=true
    Module.Index0Param1Write=true
    Module.Index0Param2Write=true
    Module.Index0Param3Write=true
   )
   &ScaleRibbonWidth(Ribbon Width Scale=&FloatFromCurve(FloatCurve=DataInterface() {propertyValues {Curve {keys [
         {interpMode=RCIM_Cubic, leaveTangent=13.5619, tangentMode=RCTM_Break, time=0, value=2}
         {arriveTangent=5.7317, interpMode=RCIM_Cubic, leaveTangent=6.2413, tangentMode=RCTM_Break, time=0.1, value=3}
         {arriveTangent=-4.9129, interpMode=RCIM_Cubic, leaveTangent=-1.7691, tangentMode=RCTM_Break, time=0.6878, value=1.2059}
         {arriveTangent=-4.4465, interpMode=RCIM_Cubic, leaveTangent=2.4652, tangentMode=RCTM_Break, time=0.9964, value=0.2724}
        ]}}}))
  ]
  rendererClasses [&NiagaraRibbonRendererProperties]
  renderers [{rendererClass=&NiagaraRibbonRendererProperties}]
 }
 Ribbons {
  emitterUpdateScript [
   &EmitterState()
   &VC_RateQualityHandler(Epic Spawn Rate=90, High Spawn Rate=60, Low Spawn Rate=30, Medium Spawn Rate=45, Use Custom Rate?=true)
   &VC_SpawnRibbon_v1(Number of Ribbons=2, SpawnRate=60)
  ]
  particleSpawnScript [
   &VC_PaintedVariantsOverride() {enabled=false}
   &VC_InitializeRibbon_v1(
    Custom Offset=true
    Enable Skeletal Position=true
    Lifetime=1.35
    Local Vector Offset=&AddVectorToPosition(Position=<-10 0 0>, Vector=$User.offset)
    Offset Amount=1
    Ribbon Width=90
   )
   &AddVelocity(Velocity=<-2000 0 0>) {module=AddVelocity001}
   &Color(Color=&MakeCustomLinearColorFromBool(Bool=$User.bIsPainted, True Color=$User.PaintedColorInnerTrail, False Color=<rgba 1 1 1 1>)) {enabled=false}
   /SetVariables(Particles.ColorTag1:LinearColor=&MakeCustomLinearColorFromBool(Bool=$User.bIsPainted, True Color=$User.PaintedColorMainTrail, False Color=<rgba 0.0114 0.0006 0.0312 1>)) {isSetParameters=true, module=SetVariables_#4}
  ]
  particleUpdateScript [
   &FN_DisableTireFX(Mask Index Attribute:NiagaraInt32=$Particles.SpawnGroup, Mask Index FL=2, Mask Index FR=3, Mask Index RL=0, Mask Index RR=1)
   &InheritSourceMovement(
    Applied Position Scale=&VectorFromFloat(Value=&FloatFromCurve(FloatCurve=DataInterface() {propertyValues {Curve {keys [<at 0.3 1 RCIM_Cubic>, {tangentMode=RCTM_User, time=1, value=0}]}}}))
    ApplyForce=false
    ApplyPosition=true
    ApplyPositionFalloff=false
   )
   &ParticleState()
   &ScaleColor(
    Scale Alpha=&FloatFromCurve(FloatCurve=DataInterface() {propertyValues {Curve {keys [
         {arriveTangent=15.4706, interpMode=RCIM_Cubic, leaveTangent=15.4706, tangentMode=RCTM_User, time=0, value=0}
         {arriveTangentWeight=0.0786, interpMode=RCIM_Cubic, leaveTangentWeight=0.2359, tangentMode=RCTM_User, time=0.1, value=1}
         {arriveTangentWeight=0.3037, interpMode=RCIM_Cubic, tangentMode=RCTM_User, time=1, value=0}
        ]}}})
    Scale RGB=&VectorFromFloat(Value=1)
    ScaleA=true
    ScaleRGB=true
   )
   &ScaleRibbonWidth(Ribbon Width Scale=&FloatFromCurve(
     FloatCurve=DataInterface() {propertyValues {Curve {keys [<at 0 1 RCIM_Cubic>, {arriveTangent=-1.3761, tangentMode=RCTM_User, time=0.4, value=0}]}}}
     Scale Curve=2
    ))
   &SolveForcesAndVelocity(Clamp Velocity=false, Limit Acceleration=false)
   &DynamicMaterialParameters(
    Index 0 Param 1=0.8
    Index 0 Param 2=&MakeCustomFloatFromBool(Bool=$User.bIsPainted, True Float=1, False Float=1.1)
    Index 0 Param 3=1
    Index 0 Param 4=&MakeCustomFloatFromBool(Bool=$User.bIsPainted, True Float=1, False Float=0)
    Module.Index0Param0Write=true
    Module.Index0Param1Write=true
    Module.Index0Param2Write=true
    Module.Index0Param3Write=true
   )
   &VC_UpdateRibbon_v1(Enable Skeletal Position=true)
   /SetVariables(
    Particles.Index0:Vector4f=&MakeVector4(X=1, Y=0, Z=0, W=0)
    Particles.MainTrailColor:LinearColor=&VC_DynamicScaleColor(
     Color to Scale:LinearColor=$Particles.ColorTag1
     RGB Scale:Vector3f=&VectorFromFloat(Value=&FloatFromCurve(FloatCurve=DataInterface() {propertyValues {Curve {keys [<at 0 5 RCIM_Cubic>,<at 0.4941 2.5718 RCIM_Cubic>]}}}))
     Alpha Scale:NiagaraFloat=&FloatFromCurve(
      FloatCurve=DataInterface() {propertyValues {Curve {keys [
          <at 0 0 RCIM_Cubic>
          {arriveTangent=-0.793, interpMode=RCIM_Cubic, leaveTangent=-0.793, tangentMode=RCTM_User, time=0.1, value=1}
          {arriveTangent=-0.8936, interpMode=RCIM_Cubic, leaveTangent=-0.8936, time=0.7031, value=0.6882}
          <at 0.8372 0.3412 RCIM_Cubic>
         ]}}}
      Scale Curve=&MakeCustomFloatFromBool(Bool=$User.bIsFrontendPreview, True Float=2.5, False Float=1)
     )
    )
    Particles.MainTrailWidth=80
   ) {isSetParameters=true, module=SetVariables_#5}
  ]
  rendererClasses [&NiagaraRibbonRendererProperties]
  renderers [
   {rendererClass=&NiagaraRibbonRendererProperties}
   {rendererClass=&NiagaraRibbonRendererProperties, rendererIndex=1}
  ]
 }
 Ribbons001 {
  bEnabled=false
  emitterUpdateScript [
   &EmitterState()
   &VC_RateQualityHandler(Epic Spawn Rate=90, High Spawn Rate=60, Low Spawn Rate=30, Medium Spawn Rate=45, Use Custom Rate?=true)
   &VC_SpawnRibbon_v1(Number of Ribbons=2, SpawnRate=$Emitter.Spawn Rate)
  ]
  particleSpawnScript [
   &VC_PaintedVariantsOverride() {enabled=false}
   &VC_InitializeRibbon_v1(
    Custom Offset=true
    Enable Skeletal Position=true
    Lifetime=1
    Local Vector Offset=&AddVectorToPosition(Position=<-10 0 0>, Vector=$User.offset)
    Offset Amount=1
    Ribbon Width=50
   )
   &AddVelocity(Velocity=<-2000 0 0>) {module=AddVelocity001}
   &Color(Color=&MakeCustomLinearColorFromBool(Bool=$User.bIsPainted, True Color=$User.PaintedColorInnerTrail, False Color=<rgba 1 1 1 1>))
   /SetVariables(Particles.ColorTag1:LinearColor=&MakeCustomLinearColorFromBool(Bool=$User.bIsPainted, True Color=$User.PaintedColorMainTrail, False Color=<rgba 0.0021 0.0042 0.0417 1>)) {isSetParameters=true, module=SetVariables_#4}
  ]
  particleUpdateScript [
   &FN_DisableTireFX(Mask Index Attribute:NiagaraInt32=$Particles.SpawnGroup, Mask Index FL=2, Mask Index FR=3, Mask Index RL=0, Mask Index RR=1)
   &InheritSourceMovement(
    Applied Position Scale=&VectorFromFloat(Value=&FloatFromCurve(FloatCurve=DataInterface() {propertyValues {Curve {keys [<at 0.3 1 RCIM_Cubic>, {tangentMode=RCTM_User, time=1, value=0}]}}}))
    ApplyForce=false
    ApplyPosition=true
    ApplyPositionFalloff=false
   )
   &ParticleState()
   &ScaleColor(
    Scale Alpha=&FloatFromCurve(FloatCurve=DataInterface() {propertyValues {Curve {keys [
         {arriveTangent=15.4706, interpMode=RCIM_Cubic, leaveTangent=15.4706, tangentMode=RCTM_User, time=0, value=0}
         {arriveTangentWeight=0.0786, interpMode=RCIM_Cubic, leaveTangentWeight=0.2359, tangentMode=RCTM_User, time=0.1, value=1}
         {arriveTangentWeight=0.3037, interpMode=RCIM_Cubic, tangentMode=RCTM_User, time=0.2, value=0}
        ]}}})
    Scale RGB=&VectorFromFloat(Value=1)
    ScaleA=true
    ScaleRGB=true
   )
   &ScaleRibbonWidth(Ribbon Width Scale=&FloatFromCurve(
     FloatCurve=DataInterface() {propertyValues {Curve {keys [<at 0 1 RCIM_Cubic>, {arriveTangent=-1.3761, tangentMode=RCTM_User, time=0.4, value=0}]}}}
     Scale Curve=2
    ))
   &SolveForcesAndVelocity(Clamp Velocity=false, Limit Acceleration=false)
   &DynamicMaterialParameters(
    Index 0 Param 1=0.8
    Index 0 Param 2=&MakeCustomFloatFromBool(Bool=$User.bIsPainted, True Float=1, False Float=1.1)
    Index 0 Param 3=1
    Index 0 Param 4=&MakeCustomFloatFromBool(Bool=$User.bIsPainted, True Float=1, False Float=0)
    Module.Index0Param0Write=true
    Module.Index0Param1Write=true
    Module.Index0Param2Write=true
    Module.Index0Param3Write=true
   )
   &VC_UpdateRibbon_v1(Enable Skeletal Position=true)
   /SetVariables(
    Particles.Index0:Vector4f=&MakeVector4(X=1, Y=0, Z=0, W=0)
    Particles.MainTrailColor:LinearColor=&VC_DynamicScaleColor(
     Color to Scale:LinearColor=$Particles.ColorTag1
     RGB Scale:Vector3f=&VectorFromFloat(Value=&FloatFromCurve(FloatCurve=DataInterface() {propertyValues {Curve {keys [<at 0 5 RCIM_Cubic>,<at 0.5 0.2 RCIM_Cubic>]}}}))
     Alpha Scale:NiagaraFloat=&FloatFromCurve(
      FloatCurve=DataInterface() {propertyValues {Curve {keys [
          <at 0 0 RCIM_Cubic>
          {arriveTangent=-0.793, interpMode=RCIM_Cubic, leaveTangent=-0.793, tangentMode=RCTM_User, time=0.1, value=1}
          {arriveTangent=-1.4286, interpMode=RCIM_Cubic, leaveTangent=-1.4286, time=0.7, value=0.5}
          <at 0.8 0 RCIM_Cubic>
         ]}}}
      Scale Curve=&MakeCustomFloatFromBool(Bool=$User.bIsFrontendPreview, True Float=2.5, False Float=1)
     )
    )
    Particles.MainTrailWidth=45
   ) {isSetParameters=true, module=SetVariables_#5}
  ]
  rendererClasses [&NiagaraRibbonRendererProperties]
  renderers [
   {rendererClass=&NiagaraRibbonRendererProperties}
   {rendererClass=&NiagaraRibbonRendererProperties, rendererIndex=1}
  ]
 }
 White {
  emitterUpdateScript [
   &EmitterState()
   &VC_FilteredSocketCount()
   &VC_SpawnRibbon_v1(Number of Ribbons=2, SpawnRate=60)
  ]
  particleSpawnScript [
   &VC_PaintedVariantsOverride() {enabled=false}
   &VC_InitializeRibbon_v1(
    Color=<rgba 0.0275 0.0157 0.0941 1>
    Custom Offset=true
    Enable Skeletal Position=true
    Lifetime=1.6
    Local Vector Offset=&AddVectorToPosition(Position=<0 0 0>, Vector=$User.DTCarOffset)
    Offset Amount=1
    Ribbon Width:NiagaraFloat=&Multiply_Float(A=4, B=5.5)
   )
   &AddVelocity(Velocity=<-1000 0 0>) {module=AddVelocity001}
  ]
  particleUpdateScript [
   &ParticleState()
   &InheritSourceMovement(
    Applied Position Scale=&Multiply_VectorByFloat(
     Vector=<1 1 1>
     Float=&FloatFromCurve(FloatCurve=DataInterface() {propertyValues {Curve {keys [
          {interpMode=RCIM_Cubic, leaveTangent=-0.0051, tangentMode=RCTM_Break, time=0, value=1}
          {arriveTangent=-0.247, interpMode=RCIM_Cubic, tangentMode=RCTM_Break, time=1, value=0.9}
         ]}}})
    )
    ApplyForce=false
    ApplyPosition=true
    ApplyPositionFalloff=false
   )
   &SolveForcesAndVelocity(Clamp Velocity=false, Limit Acceleration=false)
   &VC_UpdateRibbon_v1(Enable Skeletal Position=true)
   &ScaleColor(
    Scale Alpha=&FloatFromCurve(FloatCurve=DataInterface() {propertyValues {Curve {keys [
         {interpMode=RCIM_Cubic, leaveTangent=1.2518, tangentMode=RCTM_Break, time=0, value=0}
         {interpMode=RCIM_Cubic, tangentMode=RCTM_Break, time=0.5, value=1}
         {interpMode=RCIM_Cubic, tangentMode=RCTM_Break, time=1, value=-0.2}
        ]}}})
    Scale RGB=<2.5 2.5 2.5>
    ScaleA=true
    ScaleRGB=true
   )
   &DynamicMaterialParameters(
    Index 0 Param 1=&SelectFloatFromArray(Direct Array Index=$Particles.SpawnGroup, Float Selection Array=DataInterface() {propertyValues {FloatData [0,0.5]}})
    Index 0 Param 2=-2
    Index 0 Param 3=1
    Index 0 Param 4=4
    Module.Index0Param0Write=true
    Module.Index0Param1Write=true
    Module.Index0Param2Write=true
    Module.Index0Param3Write=true
   )
   &ScaleRibbonWidth(Ribbon Width Scale=&FloatFromCurve(FloatCurve=DataInterface() {propertyValues {Curve {keys [
         {interpMode=RCIM_Cubic, leaveTangent=13.5619, tangentMode=RCTM_Break, time=0, value=2}
         {arriveTangent=5.7317, interpMode=RCIM_Cubic, leaveTangent=6.2413, tangentMode=RCTM_Break, time=0.1, value=3}
         {arriveTangent=-4.9129, interpMode=RCIM_Cubic, leaveTangent=-1.7691, tangentMode=RCTM_Break, time=0.7127, value=1.2448}
         {
          arriveTangent=-4.4465
          arriveTangentWeight=0.0946
          interpMode=RCIM_Cubic
          leaveTangent=2.4652
          leaveTangentWeight=0.0012
          tangentMode=RCTM_Break
          time=0.9964
          value=0.2724
         }
        ]}}}))
  ]
  rendererClasses [&NiagaraRibbonRendererProperties]
  renderers [{rendererClass=&NiagaraRibbonRendererProperties}]
 }
}
summary {
 systemName=NS_DriftTrail_GalacticMagic
 emitters [
  {
   bEnabled=false
   emitterName=NE_Trail
   rendererClasses [&NiagaraMeshRendererProperties, &NiagaraRibbonRendererProperties]
  }
  {bEnabled=false, emitterName=NE_VC_DriftTrail_WheelWrap_R001, rendererClasses [&NiagaraMeshRendererProperties]}
  {emitterName=Ribbons, rendererClasses [&NiagaraRibbonRendererProperties]}
  {emitterName=NE_VC_Ribbon_Parent_Trail4, rendererClasses [&NiagaraRibbonRendererProperties]}
  {bEnabled=false, emitterName=NE_VC_Boost_Ribbon_Parent002, rendererClasses [&NiagaraRibbonRendererProperties]}
  {bEnabled=false, emitterName=Ribbons001, rendererClasses [&NiagaraRibbonRendererProperties]}
  {emitterName=NE_VC_Boost_Ribbon_Parent, rendererClasses [&NiagaraRibbonRendererProperties]}
  {bEnabled=false, emitterName=NE_VC_Boost_Ribbon_Parent001, rendererClasses [&NiagaraRibbonRendererProperties]}
  {emitterName=NE_VC_Boost_Ribbon_Parent003, rendererClasses [&NiagaraRibbonRendererProperties]}
  {emitterName=Red_Blue, rendererClasses [&NiagaraRibbonRendererProperties]}
  {emitterName=White, rendererClasses [&NiagaraRibbonRendererProperties]}
 ]
 userVariables [
  User.bIsFrontendPreview {type=NiagaraBool}
  User.bIsPainted {type=NiagaraBool}
  User.ColorVariantID {type=NiagaraInt32}
  User.ColorWhite {type=NiagaraFloat, defaultValue=0, description=""}
  User.CoreIntensity {type=NiagaraFloat, defaultValue=1, description=""}
  User.DTCarOffset {type=NiagaraPosition, description="", defaultValue {}}
  User.DTCarOffset {type=Vector3f, description="", defaultValue {struct=Vector3f, value=<0 0 0>}}
  User.DTCarOffset001 {type=Vector3f, description="", defaultValue {struct=Vector3f, value=<0 0 0>}}
  User.FadeMainColor {type=LinearColor, description="", defaultValue {struct=LinearColor, value=<rgba 1 1 1 1>}}
  User.InColorIntensity {type=NiagaraFloat, defaultValue=1, description=""}
  User.offset {type=Vector3f, description="offset from wheels(set through cosmetic template)", defaultValue {struct=Vector3f, value=<0 0 0>}}
  User.OutColorIntensity {type=NiagaraFloat, defaultValue=1, description=""}
  User.PaintedColor {type=LinearColor, description="", defaultValue {struct=LinearColor, value=<rgba 1 1 1 1>}}
  User.PaintedColorInnerTrail {type=LinearColor, description="", defaultValue {struct=LinearColor, value=<rgba 1 1 1 1>}}
  User.PaintedColorMainTrail {type=LinearColor, description="", defaultValue {struct=LinearColor, value=<rgba 1 1 1 1>}}
  User.PaintedScalar {type=NiagaraFloat, defaultValue=1, description=""}
  User.RibbonCoreIntensity {type=NiagaraFloat, defaultValue=10, description=""}
  User.RibbonMidIntensity {type=NiagaraFloat, defaultValue=5, description=""}
  User.RibbonOutIntensity {type=NiagaraFloat, defaultValue=2, description=""}
  User.RingsIntensityB {type=NiagaraFloat, defaultValue=5, description=""}
  User.RingsIntensityG {type=NiagaraFloat, defaultValue=5, description=""}
  User.RingsIntensityR {type=NiagaraFloat, defaultValue=5, description=""}
  User.RingsOpacity {type=NiagaraFloat, defaultValue=1, description=""}
  User.SingleColorIntensity {type=NiagaraFloat, defaultValue=1, description=""}
  User.VehicleFlags {type=NiagaraInt32, defaultValue=806289408}
  User.Wheel_Radius_Array {
   type=NiagaraDataInterfaceArrayFloat
   defaultValue {
    struct=NiagaraExt_VariableValue_DataInterface
    value {
     dataInterface=&self:NiagaraDataInterfaceArrayFloat_0
     dataInterfaceClass=&NiagaraDataInterfaceArrayFloat
    }
   }
  }
  User.Wheel_Width_Array {
   type=NiagaraDataInterfaceArrayFloat
   defaultValue {
    struct=NiagaraExt_VariableValue_DataInterface
    value {
     dataInterface=&self:NiagaraDataInterfaceArrayFloat_1
     dataInterfaceClass=&NiagaraDataInterfaceArrayFloat
    }
   }
  }
  User.WheelSize {type=NiagaraFloat, defaultValue=0, description=""}
  User.WheelTouchingArray {
   type=NiagaraDataInterfaceArrayFloat
   defaultValue {
    struct=NiagaraExt_VariableValue_DataInterface
    value {
     dataInterface=&self:NiagaraDataInterfaceArrayFloat_2
     dataInterfaceClass=&NiagaraDataInterfaceArrayFloat
    }
   }
  }
 ]
}
